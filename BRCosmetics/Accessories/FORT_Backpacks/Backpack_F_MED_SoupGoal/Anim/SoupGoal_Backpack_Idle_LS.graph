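# Level Sequence that animates the idle loop of the SoupGoal backpack cosmetic, a llama with a moving mouth and floppy tongue.
# A Control Rig parameter track on the Backpack skeletal mesh component keys the F_MED_SoupGoal_Pack_CtrlRig controls.
# Spine and neck translation bobs under head and jaw rotations, a four-segment wagging tongue, and eye rotation with Z-scale squash.
# Motion repeats in roughly five cycles, with matching first and last keys for looping.
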LevelSequence /BRCosmetics/Accessories/FORT_Backpacks/Backpack_F_MED_SoupGoal/Anim/SoupGoal_Backpack_Idle_LS
bindings [
 Backpack {
  id=#1
  parent_id=#2
  possessed_class=SkeletalMeshComponentBudgeted
  tracks [{
    type=MovieSceneControlRigParameterTrack
    display_name=F_MED_SoupGoal_Pack_CtrlRig
    sections [{
      type=MovieSceneControlRigParameterSection
      channels [
       llama_spine_base_ctrl.Location.X {
        type=MovieSceneScriptingFloatChannel
        default=0
        keys [[0,0],[11,-0.4339],[24,-0.6262],[30,-0.5961],[44,-0.0166],[60,0.1268],[78,-0.0073],[96,-0.4339],[109,-0.6262],[115,-0.5961],[129,-0.1025],[145,0.0409],[164,-0.104],[184,-0.4339],[197,-0.6262],[203,-0.5961],[217,-0.1025],[233,0.0409],[255,-0.1529],[275,-0.4339],[288,-0.6262],[294,-0.5961],[308,-0.1025],[324,0.0409],[348,-0.1515],[366,-0.4339],[379,-0.6262],[385,-0.5961],[398,-0.1566],[410,0.0126],[423,0.0469],[450,0]]
       }
       llama_spine_base_ctrl.Location.Y {type=MovieSceneScriptingFloatChannel, default=0, const [0,16]}
       llama_spine_base_ctrl.Location.Z {type=MovieSceneScriptingFloatChannel, default=0, const [0,16]}
       llama_spine_base_ctrl.Rotation.X {type=MovieSceneScriptingFloatChannel, default=0, const [0,16]}
       llama_spine_base_ctrl.Rotation.Y {type=MovieSceneScriptingFloatChannel, default=0, const [0,16]}
       llama_spine_base_ctrl.Rotation.Z {type=MovieSceneScriptingFloatChannel, default=0, const [0,16]}
       llama_spine_base_ctrl.Scale.X {type=MovieSceneScriptingFloatChannel, default=1, const [1,41]}
       llama_spine_base_ctrl.Scale.Y {type=MovieSceneScriptingFloatChannel, default=1, const [1,41]}
       llama_spine_base_ctrl.Scale.Z {type=MovieSceneScriptingFloatChannel, default=1, const [1,41]}
       llama_neck_01_ctrl.Location.X {type=MovieSceneScriptingFloatChannel, default=0, const [0,15]}
       llama_neck_01_ctrl.Location.Y {type=MovieSceneScriptingFloatChannel, default=0, const [0,15]}
       llama_neck_01_ctrl.Location.Z {type=MovieSceneScriptingFloatChannel, default=0, const [0,15]}
       llama_neck_01_ctrl.Rotation.X {type=MovieSceneScriptingFloatChannel, default=0, const [0,15]}
       llama_neck_01_ctrl.Rotation.Y {type=MovieSceneScriptingFloatChannel, default=0, const [0,15]}
       llama_neck_01_ctrl.Rotation.Z {type=MovieSceneScriptingFloatChannel, default=0, const [0,15]}
       llama_neck_01_ctrl.Scale.X {type=MovieSceneScriptingFloatChannel, default=1, const [1,15]}
       llama_neck_01_ctrl.Scale.Y {type=MovieSceneScriptingFloatChannel, default=1, const [1,15]}
       llama_neck_01_ctrl.Scale.Z {type=MovieSceneScriptingFloatChannel, default=1, const [1,15]}
       llama_neck_02_ctrl.Location.X {
        type=MovieSceneScriptingFloatChannel
        default=6.1514
        keys [[0,0.8453],[7,0.6456],[35,-0.5579],[56,0.219],[77,0.574],[85,0.5473],[90,0.4307],[96,0.2111],[102,0.1296],[116,0.1202],[120,0.2544],[127,0.6427],[142,0.7838],[173,0.8796],[200,0.1711],[204,0.181],[208,0.3435],[213,0.6518],[222,0.8054],[264,0.8763],[280,0.2787],[285,0.1863],[290,0.1652],[295,0.2095],[298,0.3419],[308,0.7892],[350,0.8789],[362,0.6277],[374,0.218],[380,0.1633],[386,0.2095],[390,0.4056],[396,0.8556],[426,1.0087],[450,0.8453]]
       }
       llama_neck_02_ctrl.Location.Y {
        type=MovieSceneScriptingFloatChannel
        default=0.6574
        thinned_from=53
        keys [[0,0.0001],[11.0425,0.0097],[17,0.0278],[24,0.0566],[35,0.0698],[43,0.0632],[56,0.0172],[70,-0.0097],[83,-0.0165],[92,-0.0034],[100,0.0212],[106,0.0544],[117,0.0736],[127,0.0571],[134,0.0326],[147,-0.0211],[161,-0.0385],[172,-0.0182],[180,0.0129],[184,0.0289],[189,0.0483],[197,0.0717],[208,0.0816],[215,0.0676],[219,0.0522],[228,0.0067],[239,-0.0309],[244,-0.0383],[253,-0.0332],[268,-0.0156],[280,0.0168],[283,0.0346],[285,0.0501],[291,0.0772],[299,0.0882],[306,0.0835],[310,0.0748],[319,0.0411],[323,0.0244],[335,-0.0171],[341,-0.0229],[350,-0.0169],[355,-0.006],[366,0.0231],[376,0.0605],[390,0.0882],[397,0.0835],[425,-0.0056],[430,-0.0107],[448,-0.0002],[450,0.0001]]
       }
       llama_neck_02_ctrl.Location.Z {
        type=MovieSceneScriptingFloatChannel
        default=-0.2164
        thinned_from=46
        keys [[0,1.5162],[5,1.6596],[11.0425,2.0962],[17,2.5672],[27,2.8277],[35,2.6482],[43,1.8565],[45,1.6901],[56,1.2298],[71,1.1225],[78,1.1563],[87,1.3666],[100,1.952],[112,2.1582],[121,2.0195],[144,1.3424],[156,1.2416],[165,1.295],[180,1.6167],[189,1.9914],[200,2.1641],[206,2.0679],[219,1.5312],[230,1.2956],[243,1.2237],[254,1.2871],[267,1.4754],[275,1.757],[280,1.9763],[291,2.158],[299,1.9792],[319,1.3694],[334,1.281],[350,1.3578],[362,1.6028],[366,1.764],[371,1.9789],[382,2.1638],[386,2.1393],[390,2.0786],[397,1.7385],[410,1.3694],[414,1.325],[426,1.2807],[450,1.5162]]
       }
       llama_neck_02_ctrl.Rotation.X {type=MovieSceneScriptingFloatChannel, default=0, const [0,1]}
       llama_neck_02_ctrl.Rotation.Y {type=MovieSceneScriptingFloatChannel, default=0, thinned_from=38, keys [[0,0],[450,0]]}
       llama_neck_02_ctrl.Rotation.Z {type=MovieSceneScriptingFloatChannel, default=0, const [0.0027,1]}
       llama_neck_02_ctrl.Scale.X {type=MovieSceneScriptingFloatChannel, default=1, const [1,2]}
       llama_neck_02_ctrl.Scale.Y {type=MovieSceneScriptingFloatChannel, default=1, const [1,2]}
       llama_neck_02_ctrl.Scale.Z {type=MovieSceneScriptingFloatChannel, default=1, const [1,2]}
       llama_head_ctrl.Location.X {type=MovieSceneScriptingFloatChannel, default=0, const [0,15]}
       llama_head_ctrl.Location.Y {type=MovieSceneScriptingFloatChannel, default=0, const [0,15]}
       llama_head_ctrl.Location.Z {type=MovieSceneScriptingFloatChannel, default=0, const [0,15]}
       llama_head_ctrl.Rotation.X {type=MovieSceneScriptingFloatChannel, default=0, thinned_from=15, keys [[0,0],[450,0]]}
       llama_head_ctrl.Rotation.Y {
        type=MovieSceneScriptingFloatChannel
        default=0
        thinned_from=50
        keys [[0,-6.5367],[4,-6.7032],[18,-9.3498],[30,-8.6356],[43,-6.1349],[52,-5.2713],[65,-7.1923],[77,-7.8143],[78,-7.8083],[85,-6.5367],[89,-6.7032],[103,-9.3498],[115,-8.6356],[128,-6.1349],[137,-5.2713],[150,-7.1923],[162,-7.8143],[166,-7.7355],[171,-6.9098],[176,-6.5367],[177,-6.7032],[191,-9.3498],[203,-8.6356],[216,-6.1349],[225,-5.2713],[238,-7.1923],[250,-7.8143],[262,-6.9098],[264,-6.5367],[268,-6.7032],[282,-9.3498],[294,-8.6356],[307,-6.1349],[316,-5.2713],[329,-7.1923],[341,-7.8143],[345,-7.7355],[350,-6.9098],[355,-6.5367],[359,-6.7032],[373,-9.3498],[385,-8.6356],[398,-6.1349],[407,-5.2713],[420,-7.1923],[432,-7.8143],[445,-6.9098],[446,-6.9098],[450,-6.5367]]
       }
       llama_head_ctrl.Rotation.Z {type=MovieSceneScriptingFloatChannel, default=0, const [0.1792,15]}
       llama_head_ctrl.Scale.X {type=MovieSceneScriptingFloatChannel, default=1, const [1,13]}
       llama_head_ctrl.Scale.Y {type=MovieSceneScriptingFloatChannel, default=1, const [1,13]}
       llama_head_ctrl.Scale.Z {type=MovieSceneScriptingFloatChannel, default=1, const [1,13]}
       llama_jaw_bottom_base_ctrl.Location.X {type=MovieSceneScriptingFloatChannel, default=0, const [0,21]}
       llama_jaw_bottom_base_ctrl.Location.Y {type=MovieSceneScriptingFloatChannel, default=0, const [0,21]}
       llama_jaw_bottom_base_ctrl.Location.Z {type=MovieSceneScriptingFloatChannel, default=0, const [0,21]}
       llama_jaw_bottom_base_ctrl.Rotation.X {type=MovieSceneScriptingFloatChannel, default=0, const [0,21]}
       llama_jaw_bottom_base_ctrl.Rotation.Y {
        type=MovieSceneScriptingFloatChannel
        default=0
        thinned_from=52
        keys [[0,0],[9,-8.0307],[14,-9.3897],[32,10.5624],[39,17.6178],[46,18.6046],[56,16.1316],[74,6.3243],[78,5.4083],[83,3.515],[85,3.1487],[94,0.7049],[99,0.286],[117,6.5475],[124,9.856],[131,10.3187],[141,9.1591],[159,4.383],[166,4.0757],[171,3.6316],[176,3.1487],[182,0.7049],[187,0.286],[193,1.1967],[205,6.5475],[212,9.856],[219,10.3187],[229,9.1591],[247,4.383],[262,3.515],[264,3.1487],[273,0.7049],[278,0.286],[296,6.5475],[303,9.856],[310,10.3187],[320,9.1591],[338,4.383],[345,4.0757],[350,3.6316],[355,3.1487],[364,0.7049],[369,0.286],[387,6.5475],[394,9.856],[401,10.3187],[411,9.1591],[429,4.5602],[445,2.4746],[446,2.4746],[450,0]]
       }
       llama_jaw_bottom_base_ctrl.Rotation.Z {type=MovieSceneScriptingFloatChannel, default=0, const [0,21]}
       llama_jaw_bottom_base_ctrl.Scale.X {type=MovieSceneScriptingFloatChannel, default=1, const [1,21]}
       llama_jaw_bottom_base_ctrl.Scale.Y {type=MovieSceneScriptingFloatChannel, default=1, const [1,21]}
       llama_jaw_bottom_base_ctrl.Scale.Z {type=MovieSceneScriptingFloatChannel, default=1, const [1,21]}
       llama_jaw_bottom_01_ctrl.Location.X {
        type=MovieSceneScriptingFloatChannel
        default=0
        thinned_from=31
        keys [[0,0],[14,0],[23,-0.1413],[50,-0.1413],[73,-0.4313],[83,-0.4313],[99,-0.3517],[108,-0.4313],[174,-0.4313],[187,-0.3517],[196,-0.4313],[262,-0.4313],[278,-0.3517],[287,-0.4313],[353,-0.4313],[369,-0.3517],[378,-0.4313],[405,-0.4313],[428,-0.1413],[446,-0.1413],[450,0]]
       }
       llama_jaw_bottom_01_ctrl.Location.Y {
        type=MovieSceneScriptingFloatChannel
        default=0
        thinned_from=31
        keys [[0,0],[14,0],[23,0.0688],[50,0.0688],[73,-0.313],[83,-0.313],[99,-0.3517],[108,-0.313],[174,-0.313],[187,-0.3517],[196,-0.313],[262,-0.313],[278,-0.3517],[287,-0.313],[353,-0.313],[369,-0.3517],[378,-0.313],[405,-0.313],[428,0.0688],[446,0.0688],[450,0]]
       }
       llama_jaw_bottom_01_ctrl.Location.Z {
        type=MovieSceneScriptingFloatChannel
        default=0
        thinned_from=31
        keys [[0,0],[14,0],[23,-0.6486],[50,-0.6486],[73,-0.7171],[83,-0.7171],[99,-0.3517],[108,-0.7171],[174,-0.7171],[187,-0.3517],[196,-0.7171],[262,-0.7171],[278,-0.3517],[287,-0.7171],[353,-0.7171],[369,-0.3517],[378,-0.7171],[405,-0.7171],[428,-0.6486],[446,-0.6486],[450,0]]
       }
       llama_jaw_bottom_01_ctrl.Rotation.X {
        type=MovieSceneScriptingFloatChannel
        default=0
        keys [[0,0],[17,-1.0235],[27,1.7905],[43,-7.3861],[75,0.4983],[99,2.3125],[110,3.4549],[128,-3.9365],[165,-1.8247],[180,2.7509],[193,4.8628],[205,-2.6694],[220,-4.6405],[266,-1.8247],[276,-3.3734],[302,0.9207],[339,-3.655],[374,-0.7688],[450,0]]
       }
       llama_jaw_bottom_01_ctrl.Rotation.Y {
        type=MovieSceneScriptingFloatChannel
        default=0
        keys [[0,0],[5,11.3328],[14,17.485],[18,14.0646],[23,-3.2859],[35,-8.6677],[54,-9.8735],[73,-3.2834],[78,-2.79],[90,6.0318],[99,9.4973],[103,7.5706],[108,-2.2026],[120,-5.2341],[139,-5.9133],[158,-3.0464],[166,-2.7274],[178,6.0318],[187,9.4973],[191,7.5706],[196,-2.2026],[208,-5.2341],[227,-5.9133],[246,-3.0464],[261,-2.4143],[269,6.0318],[278,9.4973],[282,7.5706],[287,-2.2026],[299,-5.2341],[318,-5.9133],[337,-3.0464],[345,-2.7274],[360,6.0318],[369,9.4973],[373,7.5706],[378,-1.6496],[390,-5.2341],[409,-5.9133],[450,0]]
       }
       llama_jaw_bottom_01_ctrl.Rotation.Z {
        type=MovieSceneScriptingFloatChannel
        default=0
        thinned_from=33
        keys [[0,0],[14,-1.8519],[18,-1.5251],[23,0.2527],[50,0.5701],[73,-0.174],[83,-0.174],[99,-1.3949],[103,-1.2108],[108,-0.2094],[135,-0.0306],[158,-0.174],[174,-0.174],[187,-1.3949],[191,-1.2108],[196,-0.2094],[223,-0.0306],[246,-0.174],[262,-0.174],[278,-1.3949],[282,-1.2108],[287,-0.2094],[314,-0.0306],[337,-0.174],[353,-0.174],[369,-1.3949],[373,-1.2108],[378,-0.2094],[405,-0.0306],[450,0]]
       }
       llama_jaw_bottom_01_ctrl.Scale.X {type=MovieSceneScriptingFloatChannel, default=1, const [1,31]}
       llama_jaw_bottom_01_ctrl.Scale.Y {type=MovieSceneScriptingFloatChannel, default=1, const [1,31]}
       llama_jaw_bottom_01_ctrl.Scale.Z {type=MovieSceneScriptingFloatChannel, default=1, const [1,31]}
       llama_tongue_1_ctrl.Location.X {type=MovieSceneScriptingFloatChannel, default=0, thinned_from=69, keys [[0,0],[450,0]]}
       llama_tongue_1_ctrl.Location.Y {type=MovieSceneScriptingFloatChannel, default=0, thinned_from=69, keys [[0,0],[450,0]]}
       llama_tongue_1_ctrl.Location.Z {type=MovieSceneScriptingFloatChannel, default=0, thinned_from=69, keys [[0,0],[450,0]]}
       llama_tongue_1_ctrl.Rotation.X {
        type=MovieSceneScriptingFloatChannel
        default=-3.1299
        keys [[0,-3.1299],[4,-3.3356],[10,-2.0199],[15,-2.4197],[19,-2.4197],[23,-1.5036],[32,-7.1058],[38,-22.545],[43,-24.1703],[45,-25.3941],[48,-19.8332],[51,-2.5276],[53,2.5151],[59,9.5679],[61,4.7669],[65,-8.4608],[73,-18.7912],[78,-21.5195],[86,-15.3072],[96,-3.4899],[105,-2.4197],[110,-0.9778],[116,-5.8016],[117,-7.9206],[122,-22.5276],[130,-25.3941],[133,-19.8332],[136,-2.5276],[139,4.262],[142,8.4643],[144,9.5679],[146,4.7669],[150,-8.4608],[158,-18.7912],[164,-21.6312],[166,-21.6312],[171,-21.5195],[179,-15.3072],[189,-3.4899],[198,-2.4197],[203,-0.9778],[209,-5.8016],[210,-7.9206],[215,-22.5276],[223,-25.3941],[226,-19.8332],[229,-2.5276],[232,4.262],[235,8.4643],[237,9.5679],[239,4.7669],[243,-8.4608],[251,-18.7912],[257,-21.6312],[259,-21.6312],[262,-21.5195],[270,-15.3072],[280,-3.4899],[289,-2.4197],[294,-0.9778],[300,-5.8016],[301,-7.9206],[306,-22.5276],[314,-25.3941],[317,-19.8332],[320,-2.5276],[323,4.262],[326,8.4643],[328,9.5679],[330,4.7669],[334,-8.4608],[345,-18.7912],[351,-21.6312],[353,-21.6312],[358,-21.5195],[364,-15.3072],[371,-3.4899],[380,-2.4197],[385,-0.9778],[391,-5.8016],[392,-7.9206],[397,-22.5276],[405,-25.3941],[408,-19.8332],[411,-2.5276],[414,4.262],[417,8.4643],[419,9.5679],[421,4.7669],[425,-8.4608],[433,-18.7912],[439,-21.6312],[441,-21.6312],[445,-20.6379],[450,-3.1299]]
       }
       llama_tongue_1_ctrl.Rotation.Y {
        type=MovieSceneScriptingFloatChannel
        default=-38.2088
        keys [[0,-38.2088],[4,-40.0047],[10,-30.9568],[23,-29.2007],[32,-17.3768],[37,-18.6086],[42,-27.1323],[45,-36.9038],[48,-43.5874],[53,-38.4637],[62,-22.6664],[65,-21.8259],[70,-28.7936],[74,-25.6777],[78,-15.1692],[86,-36.6468],[96,-45.6548],[105,-31.992],[110,-19.0588],[116,-17.4258],[117,-21.5996],[122,-28.4727],[130,-36.9038],[133,-43.5874],[139,-36.7489],[147,-22.6664],[155,-28.7936],[159,-25.6777],[164,-12.744],[166,-8.9958],[171,-15.1692],[179,-36.6468],[189,-45.6548],[198,-31.992],[203,-19.0588],[209,-17.4258],[210,-21.5996],[215,-28.4727],[223,-36.9038],[226,-43.5874],[232,-36.7489],[235,-30.5852],[240,-22.6664],[243,-21.8259],[248,-28.7936],[252,-25.6777],[257,-12.744],[259,-8.9958],[262,-15.1692],[270,-36.6468],[280,-45.6548],[289,-31.992],[294,-19.0588],[300,-17.4258],[301,-21.5996],[306,-28.4727],[314,-36.9038],[317,-43.5874],[323,-36.7489],[326,-30.5852],[331,-22.6664],[334,-21.8259],[339,-28.7936],[346,-25.6777],[351,-12.744],[353,-8.9958],[358,-15.1692],[364,-36.6468],[371,-45.6548],[380,-31.992],[385,-19.0588],[391,-17.4258],[392,-21.5996],[397,-28.4727],[405,-36.9038],[408,-43.5874],[414,-36.7489],[417,-30.5852],[422,-22.6664],[425,-21.8259],[430,-28.7936],[434,-25.6777],[439,-12.744],[441,-8.9958],[445,-19.9929],[450,-38.2088]]
       }
       llama_tongue_1_ctrl.Rotation.Z {
        type=MovieSceneScriptingFloatChannel
        default=4.9579
        keys [[0,4.9579],[4,5.1995],[10,3.2352],[15,3.9123],[19,3.9123],[23,2.3061],[32,-1.0536],[38,3.9407],[45,31.7739],[48,23.4358],[53,-9.5465],[54,-10.9403],[59,-1.9615],[65,13.5888],[70,19.4596],[78,1.1277],[86,4.2926],[96,5.3251],[105,3.9123],[110,1.0725],[116,-0.9916],[117,-0.4287],[122,4.28],[130,31.7739],[133,23.4358],[139,-10.9403],[142,-7.0615],[144,-1.9615],[150,13.5888],[155,19.4596],[164,-0.1129],[166,-1.5148],[171,1.1277],[179,4.2926],[189,5.3251],[198,3.9123],[203,1.0725],[209,-0.9916],[210,-0.4287],[215,4.28],[223,31.7739],[226,23.4358],[232,-10.9403],[235,-7.0615],[237,-1.9615],[243,13.5888],[248,19.4596],[257,-0.1129],[259,-1.5148],[262,1.1277],[270,4.2926],[280,5.3251],[289,3.9123],[294,1.0725],[300,-0.9916],[301,-0.4287],[306,4.28],[314,31.7739],[317,23.4358],[323,-10.9403],[326,-7.0615],[328,-1.9615],[334,13.5888],[339,19.4596],[351,-0.1129],[353,-1.5148],[358,1.1277],[364,4.2926],[371,5.3251],[380,3.9123],[385,1.0725],[391,-0.9916],[392,-0.4287],[397,4.28],[405,31.7739],[408,23.4358],[414,-10.9403],[417,-7.0615],[419,-1.9615],[425,13.5888],[430,19.4596],[439,-0.1129],[441,-1.5148],[445,8.429],[450,4.9579]]
       }
       llama_tongue_1_ctrl.Scale.X {type=MovieSceneScriptingFloatChannel, default=1, const [1,8]}
       llama_tongue_1_ctrl.Scale.Y {type=MovieSceneScriptingFloatChannel, default=1, const [1,8]}
       llama_tongue_1_ctrl.Scale.Z {type=MovieSceneScriptingFloatChannel, default=1, const [1,8]}
       llama_tongue_2_ctrl.Location.X {type=MovieSceneScriptingFloatChannel, default=0, thinned_from=78, keys [[0,0],[450,0]]}
       llama_tongue_2_ctrl.Location.Y {type=MovieSceneScriptingFloatChannel, default=0, thinned_from=78, keys [[0,0],[450,0]]}
       llama_tongue_2_ctrl.Location.Z {type=MovieSceneScriptingFloatChannel, default=0, thinned_from=78, keys [[0,0],[450,0]]}
       llama_tongue_2_ctrl.Rotation.X {
        type=MovieSceneScriptingFloatChannel
        default=2.3664
        thinned_from=93
        keys [[0,2.3664],[4,1.4777],[10,1.3638],[15,2.3664],[24,3.9939],[32,-3.3625],[38,0.2188],[43,2.7452],[45,3.7674],[48,5.0064],[52,4.5244],[53,5.7894],[54,8.1793],[57,15.6961],[58,17.7059],[60,19.33],[67,15.0027],[78,10.6543],[86,6.1373],[96,2.265],[105,1.9598],[110,-2.055],[117,-3.3625],[123,0.2188],[128,2.7452],[130,3.7674],[133,5.0064],[137,4.5244],[139,8.1793],[142,15.6961],[143,17.7059],[145,19.33],[152,15.0027],[156,2.3664],[161,10.8374],[166,10.3797],[171,10.6543],[179,6.1373],[189,2.265],[198,1.9598],[203,-2.055],[210,-3.3625],[216,0.2188],[221,2.7452],[223,3.7674],[226,5.0064],[230,4.5244],[232,8.1793],[235,15.6961],[236,17.7059],[238,19.33],[245,15.0027],[254,10.8374],[259,10.3797],[262,10.6543],[270,6.1373],[280,2.265],[289,1.9598],[294,-2.055],[301,-3.3625],[307,0.2188],[312,2.7452],[314,3.7674],[317,5.0064],[321,4.5244],[323,8.1793],[326,15.6961],[327,17.7059],[329,19.33],[336,15.0027],[348,10.8374],[353,10.3797],[358,10.6543],[364,6.1373],[371,2.265],[380,1.9598],[385,-2.055],[392,-3.3625],[398,0.2188],[403,2.7452],[405,3.7674],[408,5.0064],[412,4.5244],[414,8.1793],[417,15.6961],[418,17.7059],[420,19.33],[427,15.0027],[436,10.8374],[441,10.3797],[445,9.3798],[450,2.3664]]
       }
       llama_tongue_2_ctrl.Rotation.Y {
        type=MovieSceneScriptingFloatChannel
        default=15.6132
        keys [[0,21.0929],[4,10.1517],[10,-0.9608],[24,-3.7575],[28,0.8792],[45,-3.1229],[52,-18.2765],[53,-22.1056],[54,-24.501],[57,-10.1611],[58,0.8879],[59,7.3012],[60,8.4061],[65,7.1422],[78,-4.739],[86,8.4687],[96,15.325],[105,9.4243],[110,-4.7082],[117,5.4948],[123,9.8171],[128,6.1367],[133,-4.4289],[137,-20.2229],[139,-24.501],[142,-10.1611],[143,0.8879],[144,7.3012],[145,8.4061],[148,-7.1962],[156,4.3172],[165,-5.2538],[166,-5.1303],[171,-4.739],[179,8.4687],[189,15.325],[198,9.4243],[203,-4.7082],[210,5.4948],[216,9.8171],[221,6.1367],[226,-4.4289],[230,-20.2229],[232,-24.501],[235,-10.1611],[236,0.8879],[237,7.3012],[238,8.4061],[241,-7.1962],[258,-5.2538],[259,-5.1303],[262,-4.739],[270,8.4687],[280,15.325],[289,9.4243],[294,-4.7082],[301,5.4948],[307,9.8171],[312,6.1367],[317,-4.4289],[321,-20.2229],[323,-24.501],[326,-10.1611],[327,0.8879],[328,7.3012],[329,8.4061],[332,-7.1962],[352,-5.2538],[353,-5.1303],[358,-4.739],[364,8.4687],[371,15.325],[380,9.4243],[385,-4.7082],[392,5.4948],[398,9.8171],[403,6.1367],[408,-4.4289],[412,-20.2229],[414,-24.501],[417,-10.1611],[418,0.8879],[419,7.3012],[420,8.4061],[423,-7.1962],[440,-5.2538],[441,-5.1303],[445,6.5399],[450,21.0929]]
       }
       llama_tongue_2_ctrl.Rotation.Z {
        type=MovieSceneScriptingFloatChannel
        default=2.8218
        keys [[0,2.9529],[4,3.6512],[10,3.9072],[15,2.8218],[24,-4.4543],[32,-27.6134],[40,-21.2982],[48,-4.3679],[52,16.3418],[53,15.449],[54,13.4848],[57,-9.3919],[58,-15.4696],[60,-16.5375],[65,-10.695],[70,-8.4496],[78,-8.2719],[86,-3.0532],[96,2.7151],[105,0.5301],[110,-10.1154],[117,-27.6134],[125,-21.2982],[133,-4.3679],[137,16.3418],[139,13.4848],[142,-9.3919],[143,-15.4696],[145,-16.5375],[150,-10.695],[151,-9.4081],[156,2.8218],[161,-7.432],[166,-11.4699],[171,-8.2719],[179,-3.0532],[189,2.7151],[198,0.5301],[203,-10.1154],[210,-27.6134],[218,-21.2982],[226,-4.3679],[230,16.3418],[232,13.4848],[235,-9.3919],[236,-15.4696],[238,-16.5375],[243,-10.695],[244,-9.4081],[248,-8.4496],[254,-7.432],[259,-11.4699],[262,-8.2719],[270,-3.0532],[280,2.7151],[289,0.5301],[294,-10.1154],[301,-27.6134],[309,-21.2982],[317,-4.3679],[321,16.3418],[323,13.4848],[326,-9.3919],[327,-15.4696],[329,-16.5375],[334,-10.695],[335,-9.4081],[339,-8.4496],[348,-7.432],[353,-11.4699],[358,-8.2719],[364,-3.0532],[371,2.7151],[380,0.5301],[385,-10.1154],[392,-27.6134],[400,-21.2982],[408,-4.3679],[412,16.3418],[414,13.4848],[417,-9.3919],[418,-15.4696],[420,-16.5375],[425,-10.695],[426,-9.4081],[430,-8.4496],[436,-7.432],[441,-11.4699],[445,-13.7116],[450,2.9529]]
       }
       llama_tongue_2_ctrl.Scale.X {type=MovieSceneScriptingFloatChannel, default=1, const [1,78]}
       llama_tongue_2_ctrl.Scale.Y {type=MovieSceneScriptingFloatChannel, default=1, const [1,78]}
       llama_tongue_2_ctrl.Scale.Z {type=MovieSceneScriptingFloatChannel, default=1, const [1,78]}
       llama_tongue_3_ctrl.Location.X {type=MovieSceneScriptingFloatChannel, default=0, thinned_from=72, keys [[0,0],[450,0]]}
       llama_tongue_3_ctrl.Location.Y {type=MovieSceneScriptingFloatChannel, default=0, thinned_from=72, keys [[0,0],[450,0]]}
       llama_tongue_3_ctrl.Location.Z {type=MovieSceneScriptingFloatChannel, default=0, thinned_from=72, keys [[0,0],[450,0]]}
       llama_tongue_3_ctrl.Rotation.X {
        type=MovieSceneScriptingFloatChannel
        default=0
        keys [[0,9.5116],[4,8.9504],[10,4.2132],[15,7.904],[23,-6.7239],[27,-12.1329],[32,-2.2817],[38,-5.8605],[43,-8.2619],[45,-7.5384],[53,-1.2567],[59,1.576],[65,-1.5502],[78,11.8378],[86,20.6003],[96,14.2412],[105,-0.546],[110,-11.5062],[123,-5.8605],[128,-8.2619],[130,-7.5384],[139,-0.555],[142,1.1328],[144,1.576],[145,1.3445],[147,0.0754],[150,-1.5502],[161,13.7374],[166,8.0379],[171,11.8378],[179,20.6003],[189,14.2412],[198,-0.546],[203,-11.5062],[216,-5.8605],[221,-8.2619],[223,-7.5384],[232,-0.555],[235,1.1328],[237,1.576],[238,1.3444],[240,0.0754],[243,-1.5502],[254,13.7374],[259,8.0379],[262,11.8378],[270,20.6003],[280,14.2412],[289,-0.546],[294,-11.5062],[307,-5.8605],[312,-8.2619],[314,-7.5384],[323,-0.555],[326,1.1328],[328,1.576],[329,1.3444],[331,0.0754],[334,-1.5502],[348,13.7374],[353,8.0379],[358,11.8378],[364,20.6003],[371,14.2412],[380,-0.546],[385,-11.5062],[398,-5.8605],[403,-8.2619],[405,-7.5384],[414,-0.555],[417,1.1328],[419,1.576],[420,1.3444],[422,0.0754],[425,-1.5502],[436,13.7374],[441,8.0379],[445,4.9634],[450,9.5116]]
       }
       llama_tongue_3_ctrl.Rotation.Y {
        type=MovieSceneScriptingFloatChannel
        default=0
        keys [[0,35.6901],[4,34.5213],[10,25.0719],[15,32.1738],[23,16.358],[27,3.5372],[36,17.6921],[43,31.8239],[45,36.5149],[50,20.1834],[58,-2.16],[59,-2.921],[65,34.9102],[78,26.8086],[86,40.8564],[96,27.76],[105,-8.9088],[110,-9.517],[118,22.4103],[130,36.5149],[135,20.1834],[139,6.5344],[142,-0.6599],[143,-2.16],[144,-2.921],[145,-0.1187],[146,17.6142],[147,30.9807],[148,35.5927],[150,34.9102],[163,-3.4133],[166,2.6211],[171,26.8086],[179,40.8564],[189,27.76],[198,-8.9088],[203,-9.517],[211,22.4103],[223,36.5149],[228,20.1834],[232,6.5344],[235,-0.6599],[236,-2.16],[237,-2.921],[238,-0.1187],[239,17.6142],[240,30.9807],[241,35.5927],[243,34.9102],[256,-3.4133],[259,2.6211],[262,26.8086],[270,40.8564],[280,27.76],[289,-8.9088],[294,-9.517],[302,22.4103],[314,36.5149],[319,20.1834],[323,6.5344],[326,-0.6599],[327,-2.16],[328,-2.921],[329,-0.1187],[330,17.6142],[331,30.9807],[332,35.5927],[334,34.9102],[350,-3.4133],[353,2.6211],[358,26.8086],[364,40.8564],[371,27.76],[380,-8.9088],[385,-9.517],[393,22.4103],[405,36.5149],[410,20.1834],[414,6.5344],[417,-0.6599],[418,-2.16],[419,-2.921],[420,-0.1187],[421,17.6142],[422,30.9807],[423,35.5927],[425,34.9102],[438,-3.4133],[441,2.6211],[445,27.0939],[450,35.6901]]
       }
       llama_tongue_3_ctrl.Rotation.Z {
        type=MovieSceneScriptingFloatChannel
        default=0
        keys [[0,20.9295],[4,21.3986],[10,22.5539],[15,24.5783],[23,18.8753],[27,8.443],[32,-8.8661],[38,-9.6699],[43,-6.0167],[45,-4.5676],[48,11.4647],[53,43.6633],[57,22.0358],[59,9.7985],[65,3.8374],[72,11.0113],[80,1.513],[86,-1.2099],[93,6.0687],[105,23.3123],[110,15.9389],[125,-14.2536],[130,-4.5676],[133,11.4647],[139,41.8047],[142,22.0358],[144,9.7985],[145,8.322],[146,6.765],[147,5.69],[150,3.8374],[161,18.9155],[166,14.3006],[173,-9.6802],[179,-14.7189],[192,25.3672],[198,23.3123],[203,15.9389],[218,-14.2536],[223,-4.5676],[226,11.4647],[232,41.8047],[235,22.0358],[237,9.7985],[238,8.322],[239,6.765],[240,5.69],[243,3.8374],[254,18.9155],[259,14.3006],[266,-10.8381],[270,-12.4031],[283,25.3672],[289,23.3123],[294,15.9389],[309,-14.2536],[314,-4.5676],[317,11.4647],[323,41.8047],[326,22.0358],[328,9.7985],[329,8.322],[330,6.765],[331,5.69],[334,3.8374],[348,18.9155],[353,14.3006],[359,-7.7503],[364,-12.4031],[374,25.3672],[380,23.3123],[385,15.9389],[400,-14.2536],[405,-4.5676],[408,11.4647],[414,41.8047],[417,22.0358],[419,9.7985],[420,8.322],[421,6.765],[422,5.69],[425,3.8374],[436,18.9155],[441,10.0549],[445,5.541],[450,20.9295]]
       }
       llama_tongue_3_ctrl.Scale.X {type=MovieSceneScriptingFloatChannel, default=1, const [1,72]}
       llama_tongue_3_ctrl.Scale.Y {type=MovieSceneScriptingFloatChannel, default=1, const [1,72]}
       llama_tongue_3_ctrl.Scale.Z {type=MovieSceneScriptingFloatChannel, default=1, const [1,72]}
       llama_tongue_4_ctrl.Location.X {type=MovieSceneScriptingFloatChannel, default=0, thinned_from=91, keys [[0,0],[450,0]]}
       llama_tongue_4_ctrl.Location.Y {type=MovieSceneScriptingFloatChannel, default=0, thinned_from=91, keys [[0,0],[450,0]]}
       llama_tongue_4_ctrl.Location.Z {type=MovieSceneScriptingFloatChannel, default=0, thinned_from=91, keys [[0,0],[450,0]]}
       llama_tongue_4_ctrl.Rotation.X {
        type=MovieSceneScriptingFloatChannel
        default=0
        keys [[0,9.3444],[4,16.2825],[10,18.4387],[15,13.678],[19,8.5914],[23,8.4799],[25,-8.7285],[26,-22.192],[27,-25.9541],[28,-23.8024],[29,-20.9507],[30,-18.9019],[31,-16.7786],[32,-14.5635],[38,-4.5164],[43,-0.8265],[48,-1.4488],[51,-7.0245],[57,-3.7202],[62,-8.1835],[65,-6.5875],[69,-1.3945],[78,7.5641],[86,19.3267],[96,12.5828],[105,20.9493],[110,-12.2837],[112,-24.7933],[114,-20.9507],[115,-18.9019],[116,-16.7786],[117,-14.5635],[123,-4.5164],[128,-0.8265],[133,-1.4488],[136,-7.0245],[139,-5.3062],[142,-3.7202],[143,-4.1844],[147,-8.1835],[148,-7.9471],[150,-6.5875],[151,-5.5877],[152,-4.3754],[155,0.183],[161,7.1022],[166,8.0238],[171,7.5641],[179,19.3267],[189,12.5828],[195,18.9067],[196,16.1215],[197,16.106],[198,18.5745],[199,15.7006],[200,9.6941],[203,-12.2837],[205,-24.7933],[207,-20.9507],[208,-18.9019],[209,-16.7786],[210,-14.5635],[216,-4.5164],[221,-0.8265],[226,-1.4488],[229,-7.0245],[232,-5.3062],[235,-3.7202],[236,-4.1844],[240,-8.1835],[241,-7.9471],[243,-6.5875],[244,-5.5877],[245,-4.3754],[248,0.183],[254,7.1022],[259,8.0238],[262,7.5641],[270,19.3267],[280,12.5828],[289,20.9493],[294,-12.2837],[296,-24.7933],[298,-20.9507],[299,-18.9019],[300,-16.7786],[301,-14.5635],[307,-4.5164],[312,-0.8265],[317,-1.4488],[320,-7.0245],[323,-5.3062],[326,-3.7202],[327,-4.1844],[331,-8.1835],[332,-7.9471],[334,-6.5875],[335,-5.5877],[336,-4.3754],[339,0.183],[348,7.1022],[353,8.0238],[358,7.5641],[364,19.3267],[371,12.5828],[380,20.9493],[385,-12.2837],[387,-24.7933],[389,-20.9507],[390,-18.9019],[391,-16.7786],[392,-14.5635],[398,-4.5164],[403,-0.8265],[408,-1.4488],[411,-7.0245],[414,-5.3062],[417,-3.7202],[418,-4.1844],[422,-8.1835],[423,-7.9471],[425,-6.5875],[426,-5.5877],[427,-4.3754],[430,0.183],[436,7.1022],[441,8.0238],[445,9.4597],[450,9.3444]]
       }
       llama_tongue_4_ctrl.Rotation.Y {
        type=MovieSceneScriptingFloatChannel
        default=0
        keys [[0,15.4221],[4,26.3808],[10,26.6273],[15,19.2837],[19,12.9905],[23,0.1088],[25,-13.1327],[26,-25.692],[27,-26.2242],[28,-24.1599],[29,-11.1789],[30,0.6617],[31,10.363],[32,15.0615],[38,30.291],[43,25.1177],[51,7.1294],[57,-11.3701],[62,-9.0835],[63,-9.1512],[64,-23.6275],[65,-24.4784],[66,-20.6184],[67,-9.0529],[68,6.9087],[69,21.9656],[78,-12.032],[86,9.9788],[96,6.4953],[103,34.8627],[108,-7.6544],[112,-19.3722],[114,-11.1789],[115,0.6617],[116,10.363],[118,22.1453],[123,30.291],[128,25.1177],[139,44.0068],[142,8.2318],[143,-15.5524],[145,-2.4917],[148,9.1699],[151,16.6842],[153,24.0024],[155,19.6892],[161,-10.8076],[166,-13.2505],[171,-12.032],[179,9.9788],[189,6.4953],[195,33.2913],[196,28.2334],[197,21.2491],[198,9.5709],[199,1.5826],[200,-4.0813],[201,-7.6544],[205,-19.3722],[207,-11.1789],[208,0.6617],[209,10.363],[211,22.1453],[216,30.291],[221,25.1177],[232,44.0068],[235,8.2318],[236,-15.5524],[238,-2.4917],[241,9.1699],[244,16.6842],[246,24.0024],[248,19.6892],[254,-10.8076],[259,-13.2505],[262,-12.032],[270,9.9788],[280,6.4953],[287,34.8627],[292,-7.6544],[296,-19.3722],[298,-11.1789],[299,0.6617],[300,10.363],[302,22.1453],[307,30.291],[312,25.1177],[323,44.0068],[326,8.2318],[327,-15.5524],[329,-2.4917],[332,9.1699],[335,16.6842],[337,24.0024],[339,19.6892],[348,-10.8076],[353,-13.2505],[358,-12.032],[364,9.9788],[371,6.4953],[378,34.8627],[383,-7.6544],[387,-11.1789],[390,0.6617],[391,10.363],[393,22.1453],[398,30.291],[403,25.1177],[414,44.0068],[417,8.2318],[418,-15.5524],[420,-2.4917],[423,9.1699],[426,16.6842],[428,24.0024],[430,19.6892],[436,-10.8076],[441,-13.2505],[445,-7.7671],[450,15.4221]]
       }
       llama_tongue_4_ctrl.Rotation.Z {
        type=MovieSceneScriptingFloatChannel
        default=0
        thinned_from=123
        keys [[0,2.6662],[4,5.1368],[10,6.7291],[15,3.8403],[19,3.2697],[23,10.927],[25,13.7306],[26,14.1452],[27,14.0308],[28,9.4932],[29,1.9981],[30,-10.5362],[31,-14.5625],[32,-15.0863],[38,-19.7418],[43,-26.0693],[48,-14.2393],[57,-9.202],[59,-8.2619],[62,-5.2926],[65,-11.8231],[69,-4.9366],[78,16.1938],[86,10.3007],[96,2.9615],[105,3.1604],[110,16.5562],[112,15.0908],[114,1.9981],[115,-10.5362],[116,-14.5625],[117,-15.0863],[123,-19.7418],[128,-26.0693],[133,-14.2393],[142,-9.202],[143,-8.592],[144,-8.2619],[147,-10.1771],[148,-13.6955],[150,-22.1064],[151,-21.3265],[152,-19.4545],[155,-8.2631],[161,14.6523],[166,17.7278],[179,-12.0856],[195,-0.3703],[196,0.4943],[197,1.808],[198,3.6753],[199,6.6061],[200,10.2309],[203,16.5562],[205,15.0908],[207,1.9981],[208,-10.5362],[209,-14.5625],[210,-15.0863],[216,-19.7418],[221,-26.0693],[226,-14.2393],[235,-9.202],[236,-8.592],[237,-8.2619],[240,-10.1771],[241,-13.6955],[243,-22.1064],[244,-21.3265],[245,-19.4545],[248,-8.2631],[254,14.6523],[259,17.7278],[270,-12.0856],[289,3.1604],[294,16.5562],[296,15.0908],[298,1.9981],[299,-10.5362],[300,-14.5625],[301,-15.0863],[307,-19.7418],[312,-26.0693],[317,-14.2393],[326,-9.202],[327,-8.592],[328,-8.2619],[331,-10.1771],[332,-13.6955],[334,-22.1064],[335,-21.3265],[336,-19.4545],[339,-8.2631],[348,14.6523],[353,17.7278],[364,-12.0856],[380,3.1604],[385,16.5562],[387,15.0908],[389,1.9981],[390,-10.5362],[391,-14.5625],[392,-15.0863],[398,-19.7418],[403,-26.0693],[408,-14.2393],[417,-9.202],[418,-8.592],[419,-8.2619],[422,-10.1771],[423,-13.6955],[425,-22.1064],[426,-21.3265],[427,-19.4545],[430,-8.2631],[436,14.6523],[441,17.7278],[445,5.2605],[450,2.6662]]
       }
       llama_tongue_4_ctrl.Scale.X {type=MovieSceneScriptingFloatChannel, default=1, const [1,91]}
       llama_tongue_4_ctrl.Scale.Y {type=MovieSceneScriptingFloatChannel, default=1, const [1,91]}
       llama_tongue_4_ctrl.Scale.Z {type=MovieSceneScriptingFloatChannel, default=1, const [1,91]}
       llama_jaw_top_base_ctrl.Location.X {type=MovieSceneScriptingFloatChannel, default=0, thinned_from=25, keys [[0,0],[450,0]]}
       llama_jaw_top_base_ctrl.Location.Y {type=MovieSceneScriptingFloatChannel, default=0, thinned_from=25, keys [[0,0],[450,0]]}
       llama_jaw_top_base_ctrl.Location.Z {type=MovieSceneScriptingFloatChannel, default=0, thinned_from=25, keys [[0,0],[450,0]]}
       llama_jaw_top_base_ctrl.Rotation.X {
        type=MovieSceneScriptingFloatChannel
        default=0
        keys [[0,0],[14,-0.0341],[24,0.0661],[35,0.1186],[46.7887,0.1102],[63,0.096],[78,0.0942],[99,-0.0341],[109,0.0661],[120,0.1186],[132,0.1102],[148,0.096],[166,0.0937],[187,-0.0341],[197,0.0661],[208,0.1186],[220,0.1102],[236,0.096],[259,0.0935],[278,-0.0341],[288,0.0661],[299,0.1186],[311,0.1102],[327,0.096],[345,0.0937],[369,-0.0341],[379,0.0661],[390,0.1186],[402,0.1102],[418,0.096],[450,0]]
       }
       llama_jaw_top_base_ctrl.Rotation.Y {
        type=MovieSceneScriptingFloatChannel
        default=0
        keys [[0,-1.947],[14,1.0901],[24,-1.9773],[35,-5.9702],[46.7887,-4.6228],[63,-2.3504],[78,-2.0175],[83,-1.9155],[99,1.0901],[109,-1.9773],[120,-5.9702],[132,-4.6228],[148,-2.3504],[166,-1.9735],[171,-1.947],[187,1.0901],[197,-1.9773],[208,-5.9702],[220,-4.6228],[236,-2.3504],[259,-1.947],[262,-1.947],[278,1.0901],[288,-1.9773],[299,-5.9702],[311,-4.6228],[327,-2.3504],[345,-1.9735],[350,-1.947],[369,1.0901],[379,-1.9773],[390,-5.9702],[402,-4.6228],[418,-2.3504],[441,-1.947],[450,-1.947]]
       }
       llama_jaw_top_base_ctrl.Rotation.Z {
        type=MovieSceneScriptingFloatChannel
        default=0
        keys [[0,0],[14,-0.1083],[24,0.1919],[35,0.5945],[46.7887,0.4564],[63,0.2237],[78,0.1936],[99,-0.1083],[109,0.1919],[120,0.5945],[132,0.4564],[148,0.2237],[166,0.1851],[187,-0.1083],[197,0.1919],[208,0.5945],[220,0.4564],[236,0.2237],[259,0.1824],[278,-0.1083],[288,0.1919],[299,0.5945],[311,0.4564],[327,0.2237],[345,0.1851],[369,-0.1083],[379,0.1919],[390,0.5945],[402,0.4564],[418,0.2237],[450,0]]
       }
       llama_jaw_top_base_ctrl.Scale.X {type=MovieSceneScriptingFloatChannel, default=1, const [1,30]}
       llama_jaw_top_base_ctrl.Scale.Y {type=MovieSceneScriptingFloatChannel, default=1, const [1,30]}
       llama_jaw_top_base_ctrl.Scale.Z {type=MovieSceneScriptingFloatChannel, default=1, const [1,30]}
       llama_jaw_top_01_ctrl.Location.X {type=MovieSceneScriptingFloatChannel, default=0, thinned_from=18, keys [[0,0],[450,0]]}
       llama_jaw_top_01_ctrl.Location.Y {type=MovieSceneScriptingFloatChannel, default=0, thinned_from=18, keys [[0,0],[450,0]]}
       llama_jaw_top_01_ctrl.Location.Z {type=MovieSceneScriptingFloatChannel, default=0, thinned_from=18, keys [[0,0],[450,0]]}
       llama_jaw_top_01_ctrl.Rotation.X {type=MovieSceneScriptingFloatChannel, default=-0.0643, thinned_from=18, keys [[0,-0.0643],[450,-0.0643]]}
       llama_jaw_top_01_ctrl.Rotation.Y {
        type=MovieSceneScriptingFloatChannel
        default=5.8008
        keys [[0,-0.0457],[16,-0.4421],[25,-0.1217],[33,1.2763],[45,1.7926],[55,1.7132],[72,1.3054],[78,1.1001],[85,0.3489],[101,0.2441],[110,0.4305],[118,1.2439],[130,1.5443],[140,1.4981],[157,1.2608],[166,1.0728],[176,0.4747],[189,0.2441],[198,0.4305],[206,1.2439],[218,1.5443],[228,1.4981],[245,1.2608],[264,0.3489],[280,0.2441],[289,0.4305],[297,1.2439],[309,1.5443],[319,1.4981],[336,1.2608],[345,1.0728],[355,0.4747],[371,0.2441],[380,0.4305],[388,1.2439],[400,1.5443],[410,1.4981],[427,1.2608],[437,1.0549],[450,-0.0457]]
       }
       llama_jaw_top_01_ctrl.Rotation.Z {type=MovieSceneScriptingFloatChannel, default=-0.6151, thinned_from=20, keys [[0,-0.6151],[450,-0.6151]]}
       llama_jaw_top_01_ctrl.Scale.X {type=MovieSceneScriptingFloatChannel, default=1, const [1,18]}
       llama_jaw_top_01_ctrl.Scale.Y {type=MovieSceneScriptingFloatChannel, default=1, const [1,18]}
       llama_jaw_top_01_ctrl.Scale.Z {type=MovieSceneScriptingFloatChannel, default=1, const [1,18]}
       eye_l_base_ctrl.Location.X {type=MovieSceneScriptingFloatChannel, default=0, const [-0.226,13]}
       eye_l_base_ctrl.Location.Y {type=MovieSceneScriptingFloatChannel, default=0, const [-0.226,13]}
       eye_l_base_ctrl.Location.Z {type=MovieSceneScriptingFloatChannel, default=0, const [-0.226,13]}
       eye_l_base_ctrl.Rotation.X {
        type=MovieSceneScriptingFloatChannel
        default=0
        thinned_from=45
        keys [[0,-2.1744],[11,-2.3435],[15,-2.6331],[23,-2.5466],[37,-2.1744],[86,-2.1744],[89,-0.548],[96,-2.3435],[100,-2.6331],[108,-2.5466],[122,-2.1744],[174,-2.1744],[177,-0.548],[184,-2.3435],[188,-2.6331],[196,-2.5466],[210,-2.1744],[265,-2.1744],[268,-0.548],[275,-2.3435],[279,-2.6331],[287,-2.5466],[301,-2.1744],[356,-2.1744],[359,-0.548],[366,-2.3435],[370,-2.6331],[378,-2.5466],[392,-2.1744],[450,-2.1744]]
       }
       eye_l_base_ctrl.Rotation.Y {
        type=MovieSceneScriptingFloatChannel
        default=0
        thinned_from=41
        keys [[0,5.4036],[11,-1.167],[15,-1.5829],[23,1.2358],[37,9.7053],[52,5.9134],[78,5.4471],[86,5.3872],[96,2.3755],[100,2.1848],[108,3.4768],[122,7.359],[137,5.6209],[166,5.3944],[174,5.3872],[184,2.3755],[188,2.1848],[196,3.4768],[210,7.359],[225,5.6209],[259,5.4036],[262,5.4036],[265,5.3872],[275,2.3755],[279,2.1848],[287,3.4768],[301,7.359],[316,5.6209],[345,5.3944],[356,5.3872],[366,2.3755],[370,2.1848],[378,3.4768],[392,7.359],[407,5.6209],[441,5.4036],[450,5.4036]]
       }
       eye_l_base_ctrl.Rotation.Z {
        type=MovieSceneScriptingFloatChannel
        default=0
        thinned_from=41
        keys [[0,-2.5987],[11,-2.7825],[15,-3.1322],[23,-3.0315],[37,-2.5987],[78,-2.5987],[86,-2.5882],[96,-2.6702],[100,-2.8263],[108,-2.7814],[122,-2.5882],[174,-2.5882],[184,-2.6702],[188,-2.8263],[196,-2.7814],[210,-2.5882],[225,-2.5882],[262,-2.5987],[265,-2.5882],[275,-2.6702],[279,-2.8263],[287,-2.7814],[301,-2.5882],[356,-2.5882],[366,-2.6702],[370,-2.8263],[378,-2.7814],[392,-2.5882],[407,-2.5882],[450,-2.5987]]
       }
       eye_l_base_ctrl.Scale.X {type=MovieSceneScriptingFloatChannel, default=1, const [1,37]}
       eye_l_base_ctrl.Scale.Y {type=MovieSceneScriptingFloatChannel, default=1, const [1,37]}
       eye_l_base_ctrl.Scale.Z {
        type=MovieSceneScriptingFloatChannel
        default=1
        thinned_from=41
        keys [[0,1.5556],[11,0.9919],[15,0.925],[23,1.2394],[37,2.1111],[52,1.7043],[78,1.5671],[86,1.5549],[96,1.3741],[100,1.3527],[108,1.4535],[122,1.7331],[137,1.6026],[166,1.5564],[174,1.5549],[184,1.3741],[188,1.3527],[196,1.4535],[210,1.7331],[225,1.6026],[259,1.5556],[265,1.5549],[275,1.3741],[279,1.3527],[287,1.4535],[301,1.7331],[316,1.6026],[345,1.5564],[356,1.5549],[366,1.3741],[370,1.3527],[378,1.4535],[392,1.7331],[407,1.6026],[441,1.5556],[450,1.5556]]
       }
       eye_r_base_ctrl.Location.X {type=MovieSceneScriptingFloatChannel, default=0, thinned_from=13, keys [[0,-0.226],[450,-0.226]]}
       eye_r_base_ctrl.Location.Y {type=MovieSceneScriptingFloatChannel, default=0, thinned_from=13, keys [[0,-0.226],[450,-0.226]]}
       eye_r_base_ctrl.Location.Z {type=MovieSceneScriptingFloatChannel, default=0, thinned_from=13, keys [[0,-0.226],[450,-0.226]]}
       eye_r_base_ctrl.Rotation.X {
        type=MovieSceneScriptingFloatChannel
        default=0
        thinned_from=45
        keys [[0,-2.1744],[11,-2.3435],[15,-2.6331],[23,-2.5466],[37,-2.1744],[86,-2.1744],[89,-0.548],[96,-2.3435],[100,-2.6331],[108,-2.5466],[122,-2.1744],[174,-2.1744],[177,-0.548],[184,-2.3435],[188,-2.6331],[196,-2.5466],[210,-2.1744],[265,-2.1744],[268,-0.548],[275,-2.3435],[279,-2.6331],[287,-2.5466],[301,-2.1744],[356,-2.1744],[359,-0.548],[366,-2.3435],[370,-2.6331],[378,-2.5466],[392,-2.1744],[450,-2.1744]]
       }
       eye_r_base_ctrl.Rotation.Y {
        type=MovieSceneScriptingFloatChannel
        default=0
        thinned_from=41
        keys [[0,5.4036],[11,-1.167],[15,-1.5829],[23,1.2358],[37,9.7053],[52,5.9134],[78,5.4471],[86,5.3872],[96,2.3755],[100,2.1848],[108,3.4768],[122,7.359],[137,5.6209],[166,5.3944],[174,5.3872],[184,2.3755],[188,2.1848],[196,3.4768],[210,7.359],[225,5.6209],[259,5.4036],[262,5.4036],[265,5.3872],[275,2.3755],[279,2.1848],[287,3.4768],[301,7.359],[316,5.6209],[345,5.3944],[356,5.3872],[366,2.3755],[370,2.1848],[378,3.4768],[392,7.359],[407,5.6209],[441,5.4036],[450,5.4036]]
       }
       eye_r_base_ctrl.Rotation.Z {
        type=MovieSceneScriptingFloatChannel
        default=0
        thinned_from=41
        keys [[0,-2.5987],[11,-2.7825],[15,-3.1322],[23,-3.0315],[37,-2.5987],[78,-2.5987],[86,-2.5882],[96,-2.6702],[100,-2.8263],[108,-2.7814],[122,-2.5882],[174,-2.5882],[184,-2.6702],[188,-2.8263],[196,-2.7814],[210,-2.5882],[225,-2.5882],[262,-2.5987],[265,-2.5882],[275,-2.6702],[279,-2.8263],[287,-2.7814],[301,-2.5882],[356,-2.5882],[366,-2.6702],[370,-2.8263],[378,-2.7814],[392,-2.5882],[407,-2.5882],[450,-2.5987]]
       }
       eye_r_base_ctrl.Scale.X {type=MovieSceneScriptingFloatChannel, default=1, const [1,37]}
       eye_r_base_ctrl.Scale.Y {type=MovieSceneScriptingFloatChannel, default=1, const [1,37]}
       eye_r_base_ctrl.Scale.Z {
        type=MovieSceneScriptingFloatChannel
        default=1
        thinned_from=41
        keys [[0,1.5556],[11,0.9919],[15,0.925],[23,1.2394],[37,2.1111],[52,1.7043],[78,1.5671],[86,1.5549],[96,1.3741],[100,1.3527],[108,1.4535],[122,1.7331],[137,1.6026],[166,1.5564],[174,1.5549],[184,1.3741],[188,1.3527],[196,1.4535],[210,1.7331],[225,1.6026],[259,1.5556],[265,1.5549],[275,1.3741],[279,1.3527],[287,1.4535],[301,1.7331],[316,1.6026],[345,1.5564],[356,1.5549],[366,1.3741],[370,1.3527],[378,1.4535],[392,1.7331],[407,1.6026],[441,1.5556],[450,1.5556]]
       }
       llama_smile_l_ctrl.Location.X {
        type=MovieSceneScriptingFloatChannel
        default=0
        keys [[0,0.0868],[7,0.6877],[12,0.7288],[28,-0.0959],[37,-0.2322],[44,-0.2582],[54,-0.1932],[72,0.3157],[83,0.3799],[92,0.4634],[97,0.4745],[108,0.3692],[121,0.0034],[129,-0.0404],[139,-0.0099],[156,0.3239],[174,0.3992],[180,0.4634],[185,0.4745],[191,0.4505],[196,0.3692],[209,0.0034],[217,-0.0404],[227,-0.0099],[245,0.3275],[262,0.3992],[271,0.4634],[276,0.4745],[287,0.3692],[300,0.0034],[308,-0.0404],[318,-0.0099],[336,0.3382],[353,0.3992],[362,0.4634],[367,0.4745],[378,0.3692],[391,0.0034],[399,-0.0404],[409,-0.0099],[450,0.0868]]
       }
       llama_smile_l_ctrl.Location.Y {type=MovieSceneScriptingFloatChannel, default=0, const [0,1]}
       llama_smile_l_ctrl.Location.Z {type=MovieSceneScriptingFloatChannel, default=0, const [0,1]}
       llama_smile_l_ctrl.Rotation.X {type=MovieSceneScriptingFloatChannel, default=0, keys [[0,0],[10,0.4563]]}
       llama_smile_l_ctrl.Rotation.Y {type=MovieSceneScriptingFloatChannel, default=0, const [0,1]}
       llama_smile_l_ctrl.Rotation.Z {type=MovieSceneScriptingFloatChannel, default=0, keys [[0,0],[10,-0.8074]]}
       llama_smile_l_ctrl.Scale.X {type=MovieSceneScriptingFloatChannel, default=1, const [1,1]}
       llama_smile_l_ctrl.Scale.Y {type=MovieSceneScriptingFloatChannel, default=1, const [1,1]}
       llama_smile_l_ctrl.Scale.Z {type=MovieSceneScriptingFloatChannel, default=1, const [1,1]}
       llama_smile_r_ctrl.Location.X {
        type=MovieSceneScriptingFloatChannel
        default=0
        keys [[0,0.0868],[7,0.6877],[12,0.7288],[28,-0.0959],[37,-0.2322],[44,-0.2582],[54,-0.1932],[72,0.3157],[83,0.3799],[92,0.4634],[97,0.4745],[108,0.3692],[121,0.0034],[129,-0.0404],[139,-0.0099],[156,0.3239],[174,0.3992],[180,0.4634],[185,0.4745],[191,0.4505],[196,0.3692],[209,0.0034],[217,-0.0404],[227,-0.0099],[245,0.3275],[262,0.3992],[271,0.4634],[276,0.4745],[287,0.3692],[300,0.0034],[308,-0.0404],[318,-0.0099],[336,0.3382],[353,0.3992],[362,0.4634],[367,0.4745],[378,0.3692],[391,0.0034],[399,-0.0404],[409,-0.0099],[450,0.0868]]
       }
       llama_smile_r_ctrl.Location.Y {type=MovieSceneScriptingFloatChannel, default=0, const [0,20]}
       llama_smile_r_ctrl.Location.Z {type=MovieSceneScriptingFloatChannel, default=0, const [0,20]}
       llama_smile_r_ctrl.Rotation.X {type=MovieSceneScriptingFloatChannel, default=0, const [0,20]}
       llama_smile_r_ctrl.Rotation.Y {type=MovieSceneScriptingFloatChannel, default=0, const [0,20]}
       llama_smile_r_ctrl.Rotation.Z {type=MovieSceneScriptingFloatChannel, default=0, thinned_from=20, keys [[0,0],[450,0]]}
       llama_smile_r_ctrl.Scale.X {type=MovieSceneScriptingFloatChannel, default=1, const [1,20]}
       llama_smile_r_ctrl.Scale.Y {type=MovieSceneScriptingFloatChannel, default=1, const [1,20]}
       llama_smile_r_ctrl.Scale.Z {type=MovieSceneScriptingFloatChannel, default=1, const [1,20]}
       llama_ear_r_ctrl.Location.X {type=MovieSceneScriptingFloatChannel, default=0, const [0,15]}
       llama_ear_r_ctrl.Location.Y {type=MovieSceneScriptingFloatChannel, default=0, const [0,15]}
       llama_ear_r_ctrl.Location.Z {type=MovieSceneScriptingFloatChannel, default=0, const [0,15]}
       llama_ear_r_ctrl.Rotation.X {type=MovieSceneScriptingFloatChannel, default=0, const [0,15]}
       llama_ear_r_ctrl.Rotation.Y {type=MovieSceneScriptingFloatChannel, default=0, const [0,15]}
       llama_ear_r_ctrl.Rotation.Z {type=MovieSceneScriptingFloatChannel, default=0, const [0,15]}
       llama_ear_r_ctrl.Scale.X {type=MovieSceneScriptingFloatChannel, default=1, const [1,15]}
       llama_ear_r_ctrl.Scale.Y {type=MovieSceneScriptingFloatChannel, default=1, const [1,15]}
       llama_ear_r_ctrl.Scale.Z {type=MovieSceneScriptingFloatChannel, default=1, const [1,15]}
       llama_ear_l_ctrl.Location.X {type=MovieSceneScriptingFloatChannel, default=0, const [0,15]}
       llama_ear_l_ctrl.Location.Y {type=MovieSceneScriptingFloatChannel, default=0, const [0,15]}
       llama_ear_l_ctrl.Location.Z {type=MovieSceneScriptingFloatChannel, default=0, const [0,15]}
       llama_ear_l_ctrl.Rotation.X {type=MovieSceneScriptingFloatChannel, default=0, const [0,15]}
       llama_ear_l_ctrl.Rotation.Y {type=MovieSceneScriptingFloatChannel, default=0, const [0,15]}
       llama_ear_l_ctrl.Rotation.Z {type=MovieSceneScriptingFloatChannel, default=0, const [0,15]}
       llama_ear_l_ctrl.Scale.X {type=MovieSceneScriptingFloatChannel, default=1, const [1,15]}
       llama_ear_l_ctrl.Scale.Y {type=MovieSceneScriptingFloatChannel, default=1, const [1,15]}
       llama_ear_l_ctrl.Scale.Z {type=MovieSceneScriptingFloatChannel, default=1, const [1,15]}
       frnt_leg_l_01_ctrl.Location.X {type=MovieSceneScriptingFloatChannel, default=0, const [0,22]}
       frnt_leg_l_01_ctrl.Location.Y {type=MovieSceneScriptingFloatChannel, default=0, const [0,22]}
       frnt_leg_l_01_ctrl.Location.Z {type=MovieSceneScriptingFloatChannel, default=0, const [0,22]}
       frnt_leg_l_01_ctrl.Rotation.X {type=MovieSceneScriptingFloatChannel, default=0, const [0,22]}
       frnt_leg_l_01_ctrl.Rotation.Y {
        type=MovieSceneScriptingFloatChannel
        default=0
        keys [[0,-23.581],[19,-47.6842],[23,-43.9735],[29,-21.3321],[34,-15.3863],[42,-17.7615],[56,-24.9041],[71,-27.9551],[85,-23.581],[104,-33.8975],[108,-32.3406],[114,-22.8406],[119,-20.3459],[127,-21.3425],[141,-24.3394],[156,-25.6195],[166,-24.7148],[192,-33.8975],[196,-32.3406],[202,-22.8406],[207,-20.3459],[215,-21.3425],[229,-24.3394],[244,-25.6195],[256,-23.7405],[270,-21.8603],[283,-33.8975],[287,-32.3406],[293,-22.8406],[298,-20.3459],[306,-21.3425],[320,-24.3394],[335,-25.6195],[355,-23.7842],[374,-33.8975],[378,-32.3406],[384,-22.8406],[389,-20.3459],[397,-21.3425],[411,-24.3394],[426,-25.6195],[450,-23.581]]
       }
       frnt_leg_l_01_ctrl.Rotation.Z {type=MovieSceneScriptingFloatChannel, default=0, const [0,22]}
       frnt_leg_l_01_ctrl.Scale.X {type=MovieSceneScriptingFloatChannel, default=1, const [1,22]}
       frnt_leg_l_01_ctrl.Scale.Y {type=MovieSceneScriptingFloatChannel, default=1, const [1,22]}
       frnt_leg_l_01_ctrl.Scale.Z {type=MovieSceneScriptingFloatChannel, default=1, const [1,22]}
       frnt_leg_l_02_ctrl.Location.X {type=MovieSceneScriptingFloatChannel, default=0, const [0,17]}
       frnt_leg_l_02_ctrl.Location.Y {type=MovieSceneScriptingFloatChannel, default=0, const [0,17]}
       frnt_leg_l_02_ctrl.Location.Z {type=MovieSceneScriptingFloatChannel, default=0, const [0,17]}
       frnt_leg_l_02_ctrl.Rotation.X {
        type=MovieSceneScriptingFloatChannel
        default=0
        thinned_from=13
        keys [[0,0],[5,0],[36,0.6023],[85,0],[90,0],[121,0.6023],[209,0.6023],[269,0],[300,0.6023],[360,0],[391,0.6023],[450,0]]
       }
       frnt_leg_l_02_ctrl.Rotation.Y {
        type=MovieSceneScriptingFloatChannel
        default=0
        thinned_from=35
        keys [[0,-4.3873],[5,2.9454],[13,14.2418],[18,7.3407],[29,-16.3279],[52,2.1532],[78,-5.5363],[90,-0.3255],[98,5.9777],[103,2.127],[114,-11.0798],[137,-0.7676],[164,-5.5673],[166,-5.5023],[186,5.9777],[191,2.127],[202,-11.0798],[225,-0.7676],[252,-5.5673],[269,-0.3255],[277,5.9777],[282,2.127],[293,-11.0798],[316,-0.7676],[350,-4.4171],[355,-4.4171],[360,-0.3255],[368,5.9777],[373,2.127],[384,-11.0798],[407,-0.7676],[434,-5.5673],[446,-4.3873],[450,-4.3873]]
       }
       frnt_leg_l_02_ctrl.Rotation.Z {type=MovieSceneScriptingFloatChannel, default=0, const [0,13]}
       frnt_leg_l_02_ctrl.Scale.X {type=MovieSceneScriptingFloatChannel, default=1, const [1,17]}
       frnt_leg_l_02_ctrl.Scale.Y {type=MovieSceneScriptingFloatChannel, default=1, const [1,17]}
       frnt_leg_l_02_ctrl.Scale.Z {type=MovieSceneScriptingFloatChannel, default=1, const [1,17]}
       frnt_finger_small_l_01_ctrl.Location.X {type=MovieSceneScriptingFloatChannel, default=0, const [0,38]}
       frnt_finger_small_l_01_ctrl.Location.Y {type=MovieSceneScriptingFloatChannel, default=0, const [0,38]}
       frnt_finger_small_l_01_ctrl.Location.Z {type=MovieSceneScriptingFloatChannel, default=0, const [0,38]}
       frnt_finger_small_l_01_ctrl.Rotation.X {type=MovieSceneScriptingFloatChannel, default=0, const [0,38]}
       frnt_finger_small_l_01_ctrl.Rotation.Y {
        type=MovieSceneScriptingFloatChannel
        default=0
        thinned_from=38
        keys [[0,-47.8921],[5,-41.1362],[13,-0.4362],[25,-59.7723],[36,-49.1537],[53,-32.0313],[77,-55.3897],[78,-55.3724],[90,-44.9502],[98,-24.8927],[110,-54.1343],[121,-48.9013],[138,-40.4632],[162,-51.9745],[166,-51.826],[178,-44.9502],[186,-24.8927],[198,-54.1343],[209,-48.9013],[226,-40.4632],[250,-51.9745],[262,-47.8921],[269,-44.9502],[277,-24.8927],[289,-54.1343],[300,-48.9013],[317,-40.4632],[341,-51.9745],[345,-51.826],[360,-44.9502],[368,-24.8927],[380,-54.1343],[391,-48.9013],[408,-40.4632],[432,-51.9745],[446,-47.8921],[450,-47.8921]]
       }
       frnt_finger_small_l_01_ctrl.Rotation.Z {type=MovieSceneScriptingFloatChannel, default=0, const [0,38]}
       frnt_finger_small_l_01_ctrl.Scale.X {type=MovieSceneScriptingFloatChannel, default=1, const [1,38]}
       frnt_finger_small_l_01_ctrl.Scale.Y {type=MovieSceneScriptingFloatChannel, default=1, const [1,38]}
       frnt_finger_small_l_01_ctrl.Scale.Z {type=MovieSceneScriptingFloatChannel, default=1, const [1,38]}
       frnt_finger_small_l_02_ctrl.Location.X {type=MovieSceneScriptingFloatChannel, default=0, const [0,42]}
       frnt_finger_small_l_02_ctrl.Location.Y {type=MovieSceneScriptingFloatChannel, default=0, const [0,42]}
       frnt_finger_small_l_02_ctrl.Location.Z {type=MovieSceneScriptingFloatChannel, default=0, const [0,42]}
       frnt_finger_small_l_02_ctrl.Rotation.X {type=MovieSceneScriptingFloatChannel, default=0, const [0,42]}
       frnt_finger_small_l_02_ctrl.Rotation.Y {
        type=MovieSceneScriptingFloatChannel
        default=0
        thinned_from=41
        keys [[0,-23.2705],[5,-41.6248],[13,24.1854],[25,-60.2164],[36,-49.5978],[53,-7.4097],[77,-30.768],[78,-30.6755],[83,-22.8333],[90,-31.384],[98,-0.7252],[110,-40.0452],[121,-35.0984],[138,-15.4443],[162,-26.3262],[166,-25.7338],[171,-22.8333],[178,-31.384],[186,-0.7252],[198,-40.0452],[209,-35.0984],[226,-15.4443],[250,-26.3262],[262,-23.2705],[269,-31.384],[277,-0.7252],[289,-40.0452],[300,-35.0984],[317,-15.4443],[341,-26.3262],[345,-25.7338],[350,-22.8333],[360,-31.384],[368,-0.7252],[380,-40.0452],[391,-35.0984],[408,-15.4443],[432,-26.3262],[446,-23.2705],[450,-23.2705]]
       }
       frnt_finger_small_l_02_ctrl.Rotation.Z {type=MovieSceneScriptingFloatChannel, default=0, const [0,42]}
       frnt_finger_small_l_02_ctrl.Scale.X {type=MovieSceneScriptingFloatChannel, default=1, const [1,42]}
       frnt_finger_small_l_02_ctrl.Scale.Y {type=MovieSceneScriptingFloatChannel, default=1, const [1,42]}
       frnt_finger_small_l_02_ctrl.Scale.Z {type=MovieSceneScriptingFloatChannel, default=1, const [1,42]}
       frnt_finger_mid_l_01_ctrl.Location.X {type=MovieSceneScriptingFloatChannel, default=0, const [0,38]}
       frnt_finger_mid_l_01_ctrl.Location.Y {type=MovieSceneScriptingFloatChannel, default=0, const [0,38]}
       frnt_finger_mid_l_01_ctrl.Location.Z {type=MovieSceneScriptingFloatChannel, default=0, const [0,38]}
       frnt_finger_mid_l_01_ctrl.Rotation.X {type=MovieSceneScriptingFloatChannel, default=0, const [0,38]}
       frnt_finger_mid_l_01_ctrl.Rotation.Y {
        type=MovieSceneScriptingFloatChannel
        default=0
        thinned_from=41
        keys [[0,-41.9751],[8,-35.2192],[20,20.787],[31,-48.9991],[42,-38.3805],[56,-24.1031],[78,-47.0018],[85,-42.0547],[93,-38.9699],[105,-13.3972],[116,-45.2619],[127,-40.4134],[141,-33.8943],[165,-44.5598],[166,-44.5092],[171,-42.0547],[181,-38.9699],[193,-13.3972],[204,-45.2619],[215,-40.4134],[229,-33.8943],[253,-44.5598],[262,-41.9751],[264,-42.0547],[272,-38.9699],[284,-13.3972],[295,-45.2619],[306,-40.4134],[320,-33.8943],[344,-44.5598],[345,-44.5092],[350,-42.0547],[363,-38.9699],[375,-13.3972],[386,-45.2619],[397,-40.4134],[411,-33.8943],[435,-44.5598],[446,-41.9751],[450,-41.9751]]
       }
       frnt_finger_mid_l_01_ctrl.Rotation.Z {type=MovieSceneScriptingFloatChannel, default=0, const [0,38]}
       frnt_finger_mid_l_01_ctrl.Scale.X {type=MovieSceneScriptingFloatChannel, default=1, const [1,38]}
       frnt_finger_mid_l_01_ctrl.Scale.Y {type=MovieSceneScriptingFloatChannel, default=1, const [1,38]}
       frnt_finger_mid_l_01_ctrl.Scale.Z {type=MovieSceneScriptingFloatChannel, default=1, const [1,38]}
       frnt_finger_mid_l_02_ctrl.Location.X {type=MovieSceneScriptingFloatChannel, default=0, const [0,38]}
       frnt_finger_mid_l_02_ctrl.Location.Y {type=MovieSceneScriptingFloatChannel, default=0, const [0,38]}
       frnt_finger_mid_l_02_ctrl.Location.Z {type=MovieSceneScriptingFloatChannel, default=0, const [0,38]}
       frnt_finger_mid_l_02_ctrl.Rotation.X {type=MovieSceneScriptingFloatChannel, default=0, const [0,38]}
       frnt_finger_mid_l_02_ctrl.Rotation.Y {
        type=MovieSceneScriptingFloatChannel
        default=0
        thinned_from=39
        keys [[0,-22.3665],[11,-15.6106],[23,40.3956],[34,-50.9226],[45,-40.304],[59,-4.4945],[78,-25.2338],[85,-22.5778],[96,-19.5348],[108,5.6921],[119,-35.4403],[130,-30.6574],[144,-14.5277],[166,-24.842],[171,-22.5778],[184,-19.5348],[196,5.6921],[207,-35.4403],[218,-30.6574],[232,-14.5277],[256,-25.049],[262,-22.3665],[264,-22.5778],[275,-19.5348],[287,5.6921],[298,-35.4403],[309,-30.6574],[323,-14.5277],[345,-24.842],[350,-22.5778],[366,-19.5348],[378,5.6921],[389,-35.4403],[400,-30.6574],[414,-14.5277],[438,-25.049],[446,-22.3665],[450,-22.3665]]
       }
       frnt_finger_mid_l_02_ctrl.Rotation.Z {type=MovieSceneScriptingFloatChannel, default=0, const [0,38]}
       frnt_finger_mid_l_02_ctrl.Scale.X {type=MovieSceneScriptingFloatChannel, default=1, const [1,38]}
       frnt_finger_mid_l_02_ctrl.Scale.Y {type=MovieSceneScriptingFloatChannel, default=1, const [1,38]}
       frnt_finger_mid_l_02_ctrl.Scale.Z {type=MovieSceneScriptingFloatChannel, default=1, const [1,38]}
       frnt_finger_big_l_01_ctrl.Location.X {type=MovieSceneScriptingFloatChannel, default=0, const [0,36]}
       frnt_finger_big_l_01_ctrl.Location.Y {type=MovieSceneScriptingFloatChannel, default=0, const [0,36]}
       frnt_finger_big_l_01_ctrl.Location.Z {type=MovieSceneScriptingFloatChannel, default=0, const [0,36]}
       frnt_finger_big_l_01_ctrl.Rotation.X {type=MovieSceneScriptingFloatChannel, default=0, const [0,36]}
       frnt_finger_big_l_01_ctrl.Rotation.Y {
        type=MovieSceneScriptingFloatChannel
        default=0
        keys [[0,-61.8567],[8,-68.9079],[18,-12.6291],[38,-83.16],[45,-75.4812],[58,-44.2242],[78,-65.8522],[93,-64.4116],[103,-41.8421],[123,-70.1271],[130,-67.0477],[143,-54.5127],[171,-61.5839],[181,-64.4116],[191,-41.8421],[211,-70.1271],[218,-67.0477],[231,-54.5127],[262,-61.8567],[272,-64.4116],[282,-41.8421],[302,-70.1271],[309,-67.0477],[322,-54.5127],[350,-61.5839],[363,-64.4116],[373,-41.8421],[393,-70.1271],[400,-67.0477],[413,-54.5127],[437,-63.8801],[450,-61.8567]]
       }
       frnt_finger_big_l_01_ctrl.Rotation.Z {type=MovieSceneScriptingFloatChannel, default=0, const [0,36]}
       frnt_finger_big_l_01_ctrl.Scale.X {type=MovieSceneScriptingFloatChannel, default=1, const [1,36]}
       frnt_finger_big_l_01_ctrl.Scale.Y {type=MovieSceneScriptingFloatChannel, default=1, const [1,36]}
       frnt_finger_big_l_01_ctrl.Scale.Z {type=MovieSceneScriptingFloatChannel, default=1, const [1,36]}
       frnt_finger_big_l_02_ctrl.Location.X {type=MovieSceneScriptingFloatChannel, default=0, const [0,37]}
       frnt_finger_big_l_02_ctrl.Location.Y {type=MovieSceneScriptingFloatChannel, default=0, const [0,37]}
       frnt_finger_big_l_02_ctrl.Location.Z {type=MovieSceneScriptingFloatChannel, default=0, const [0,37]}
       frnt_finger_big_l_02_ctrl.Rotation.X {type=MovieSceneScriptingFloatChannel, default=0, const [0,37]}
       frnt_finger_big_l_02_ctrl.Rotation.Y {
        type=MovieSceneScriptingFloatChannel
        default=0
        keys [[0,-22.463],[11,-41.5001],[21,26.7646],[41,-56.8758],[48,-49.1971],[61,-4.8305],[78,-23.3867],[96,-29.6252],[106,-2.4598],[126,-35.7439],[133,-32.6882],[146,-15.0328],[171,-22.0495],[184,-29.6252],[194,-2.4598],[214,-35.7439],[221,-32.6882],[234,-15.0328],[262,-22.463],[275,-29.6252],[285,-2.4598],[305,-35.7439],[312,-32.6882],[325,-15.0328],[350,-22.0495],[366,-29.6252],[376,-2.4598],[396,-35.7439],[403,-32.6882],[416,-15.0328],[440,-24.3281],[450,-22.463],[1094.16,-31.4407]]
       }
       frnt_finger_big_l_02_ctrl.Rotation.Z {type=MovieSceneScriptingFloatChannel, default=0, const [0,37]}
       frnt_finger_big_l_02_ctrl.Scale.X {type=MovieSceneScriptingFloatChannel, default=1, const [1,37]}
       frnt_finger_big_l_02_ctrl.Scale.Y {type=MovieSceneScriptingFloatChannel, default=1, const [1,37]}
       frnt_finger_big_l_02_ctrl.Scale.Z {type=MovieSceneScriptingFloatChannel, default=1, const [1,37]}
       frnt_leg_r_01_ctrl.Location.X {type=MovieSceneScriptingFloatChannel, default=0, const [0,13]}
       frnt_leg_r_01_ctrl.Location.Y {type=MovieSceneScriptingFloatChannel, default=0, const [0,13]}
       frnt_leg_r_01_ctrl.Location.Z {type=MovieSceneScriptingFloatChannel, default=0, const [0,13]}
       frnt_leg_r_01_ctrl.Rotation.X {type=MovieSceneScriptingFloatChannel, default=0, const [0,13]}
       frnt_leg_r_01_ctrl.Rotation.Y {
        type=MovieSceneScriptingFloatChannel
        default=0
        keys [[0,-23.581],[25,-47.6842],[29,-43.9735],[35,-21.3321],[40,-15.3863],[48,-17.7615],[62,-24.9041],[74,-27.9551],[85,-23.581],[110,-33.8975],[114,-32.3406],[120,-22.8406],[125,-20.3459],[133,-21.3425],[147,-24.3394],[159,-25.6195],[166,-25.3113],[198,-33.8975],[202,-32.3406],[208,-22.8406],[213,-20.3459],[221,-21.3425],[235,-24.3394],[253,-24.9361],[270,-18.5873],[289,-33.8975],[293,-32.3406],[299,-22.8406],[304,-20.3459],[312,-21.3425],[326,-24.3394],[338,-25.6195],[355,-23.7842],[380,-33.8975],[384,-32.3406],[390,-22.8406],[395,-20.3459],[403,-21.3425],[417,-24.3394],[429,-25.6195],[450,-23.581]]
       }
       frnt_leg_r_01_ctrl.Rotation.Z {type=MovieSceneScriptingFloatChannel, default=0, const [0,13]}
       frnt_leg_r_01_ctrl.Scale.X {type=MovieSceneScriptingFloatChannel, default=1, const [1,13]}
       frnt_leg_r_01_ctrl.Scale.Y {type=MovieSceneScriptingFloatChannel, default=1, const [1,13]}
       frnt_leg_r_01_ctrl.Scale.Z {type=MovieSceneScriptingFloatChannel, default=1, const [1,13]}
       frnt_leg_r_02_ctrl.Location.X {type=MovieSceneScriptingFloatChannel, default=0, const [0,13]}
       frnt_leg_r_02_ctrl.Location.Y {type=MovieSceneScriptingFloatChannel, default=0, const [0,13]}
       frnt_leg_r_02_ctrl.Location.Z {type=MovieSceneScriptingFloatChannel, default=0, const [0,13]}
       frnt_leg_r_02_ctrl.Rotation.X {type=MovieSceneScriptingFloatChannel, default=0, const [0,13]}
       frnt_leg_r_02_ctrl.Rotation.Y {
        type=MovieSceneScriptingFloatChannel
        default=0
        thinned_from=35
        keys [[0,-4.3873],[11,2.9454],[19,14.2418],[24,7.3407],[35,-16.3279],[58,2.1532],[78,-4.2796],[96,-0.3255],[104,5.9777],[109,2.127],[120,-11.0798],[143,-0.7676],[166,-5.2825],[184,-0.3255],[192,5.9777],[197,2.127],[208,-11.0798],[231,-0.7676],[258,-5.5673],[275,-0.3255],[283,5.9777],[288,2.127],[299,-11.0798],[322,-0.7676],[350,-4.4171],[355,-4.4171],[366,-0.3255],[374,5.9777],[379,2.127],[390,-11.0798],[413,-0.7676],[440,-5.5673],[446,-4.3873],[450,-4.3873]]
       }
       frnt_leg_r_02_ctrl.Rotation.Z {type=MovieSceneScriptingFloatChannel, default=0, const [0,13]}
       frnt_leg_r_02_ctrl.Scale.X {type=MovieSceneScriptingFloatChannel, default=1, const [1,13]}
       frnt_leg_r_02_ctrl.Scale.Y {type=MovieSceneScriptingFloatChannel, default=1, const [1,13]}
       frnt_leg_r_02_ctrl.Scale.Z {type=MovieSceneScriptingFloatChannel, default=1, const [1,13]}
       frnt_finger_big_r_01_ctrl.Location.X {type=MovieSceneScriptingFloatChannel, default=0, const [0,13]}
       frnt_finger_big_r_01_ctrl.Location.Y {type=MovieSceneScriptingFloatChannel, default=0, const [0,13]}
       frnt_finger_big_r_01_ctrl.Location.Z {type=MovieSceneScriptingFloatChannel, default=0, const [0,13]}
       frnt_finger_big_r_01_ctrl.Rotation.X {type=MovieSceneScriptingFloatChannel, default=0, const [0,13]}
       frnt_finger_big_r_01_ctrl.Rotation.Y {
        type=MovieSceneScriptingFloatChannel
        default=0
        thinned_from=37
        keys [[0,-61.8567],[14,-68.9079],[24,-12.6291],[44,-83.16],[51,-75.4812],[64,-44.2242],[78,-58.7961],[99,-63.5737],[109,-42.2205],[129,-68.9812],[136,-66.0677],[149,-54.2082],[166,-61.2488],[171,-60.8983],[187,-63.5737],[197,-42.2205],[217,-68.9812],[224,-66.0677],[237,-54.2082],[262,-61.8567],[278,-63.5737],[288,-42.2205],[308,-68.9812],[315,-66.0677],[328,-54.2082],[345,-61.2488],[350,-60.8983],[369,-63.5737],[379,-42.2205],[399,-68.9812],[406,-66.0677],[419,-54.2082],[443,-63.0708],[444,-60.8238],[446,-61.8567],[450,-61.8567]]
       }
       frnt_finger_big_r_01_ctrl.Rotation.Z {type=MovieSceneScriptingFloatChannel, default=0, const [0,13]}
       frnt_finger_big_r_01_ctrl.Scale.X {type=MovieSceneScriptingFloatChannel, default=1, const [1,13]}
       frnt_finger_big_r_01_ctrl.Scale.Y {type=MovieSceneScriptingFloatChannel, default=1, const [1,13]}
       frnt_finger_big_r_01_ctrl.Scale.Z {type=MovieSceneScriptingFloatChannel, default=1, const [1,13]}
       frnt_finger_big_r_02_ctrl.Location.X {type=MovieSceneScriptingFloatChannel, default=0, const [0,13]}
       frnt_finger_big_r_02_ctrl.Location.Y {type=MovieSceneScriptingFloatChannel, default=0, const [0,13]}
       frnt_finger_big_r_02_ctrl.Location.Z {type=MovieSceneScriptingFloatChannel, default=0, const [0,13]}
       frnt_finger_big_r_02_ctrl.Rotation.X {type=MovieSceneScriptingFloatChannel, default=0, const [0,13]}
       frnt_finger_big_r_02_ctrl.Rotation.Y {
        type=MovieSceneScriptingFloatChannel
        default=0
        thinned_from=36
        keys [[0,-22.463],[17,-41.5001],[27,26.7646],[47,-56.8758],[54,-49.1971],[67,-4.8305],[78,-12.9609],[102,-28.4585],[112,-2.5577],[132,-34.2924],[139,-31.3789],[152,-14.5455],[166,-20.0743],[171,-21.2355],[190,-28.4585],[200,-2.5577],[220,-34.2924],[227,-31.3789],[240,-14.5455],[262,-22.463],[281,-28.4585],[291,-2.5577],[311,-34.2924],[318,-31.3789],[331,-14.5455],[345,-20.0743],[350,-21.2355],[372,-28.4585],[382,-2.5577],[402,-34.2924],[409,-31.3789],[422,-14.5455],[444,-21.4302],[446,-22.463],[450,-22.463]]
       }
       frnt_finger_big_r_02_ctrl.Rotation.Z {type=MovieSceneScriptingFloatChannel, default=0, const [0,13]}
       frnt_finger_big_r_02_ctrl.Scale.X {type=MovieSceneScriptingFloatChannel, default=1, const [1,13]}
       frnt_finger_big_r_02_ctrl.Scale.Y {type=MovieSceneScriptingFloatChannel, default=1, const [1,13]}
       frnt_finger_big_r_02_ctrl.Scale.Z {type=MovieSceneScriptingFloatChannel, default=1, const [1,13]}
       frnt_finger_mid_r_01_ctrl.Location.X {type=MovieSceneScriptingFloatChannel, default=0, const [0,6]}
       frnt_finger_mid_r_01_ctrl.Location.Y {type=MovieSceneScriptingFloatChannel, default=0, const [0,6]}
       frnt_finger_mid_r_01_ctrl.Location.Z {type=MovieSceneScriptingFloatChannel, default=0, const [0,6]}
       frnt_finger_mid_r_01_ctrl.Rotation.X {type=MovieSceneScriptingFloatChannel, default=0, const [0,6]}
       frnt_finger_mid_r_01_ctrl.Rotation.Y {
        type=MovieSceneScriptingFloatChannel
        default=0
        thinned_from=40
        keys [[0,-41.9751],[14,-35.2192],[26,20.787],[37,-48.9991],[48,-38.3805],[62,-24.1031],[78,-38.6114],[83,-41.54],[85,-41.54],[99,-38.8932],[111,-16.9518],[122,-44.2918],[133,-40.1317],[147,-34.5383],[166,-42.6633],[171,-41.54],[187,-38.8932],[199,-16.9518],[210,-44.2918],[221,-40.1317],[235,-34.5383],[259,-43.6894],[262,-41.9751],[264,-41.54],[278,-38.8932],[290,-16.9518],[301,-44.2918],[312,-40.1317],[326,-34.5383],[345,-42.6633],[350,-41.54],[369,-38.8932],[381,-16.9518],[392,-44.2918],[403,-40.1317],[417,-34.5383],[441,-43.6894],[446,-41.9751],[450,-41.9751]]
       }
       frnt_finger_mid_r_01_ctrl.Rotation.Z {type=MovieSceneScriptingFloatChannel, default=0, const [0,6]}
       frnt_finger_mid_r_01_ctrl.Scale.X {type=MovieSceneScriptingFloatChannel, default=1, const [1,6]}
       frnt_finger_mid_r_01_ctrl.Scale.Y {type=MovieSceneScriptingFloatChannel, default=1, const [1,6]}
       frnt_finger_mid_r_01_ctrl.Scale.Z {type=MovieSceneScriptingFloatChannel, default=1, const [1,6]}
       frnt_finger_mid_r_02_ctrl.Location.X {type=MovieSceneScriptingFloatChannel, default=0, const [0,8]}
       frnt_finger_mid_r_02_ctrl.Location.Y {type=MovieSceneScriptingFloatChannel, default=0, const [0,8]}
       frnt_finger_mid_r_02_ctrl.Location.Z {type=MovieSceneScriptingFloatChannel, default=0, const [0,8]}
       frnt_finger_mid_r_02_ctrl.Rotation.X {type=MovieSceneScriptingFloatChannel, default=0, const [0,8]}
       frnt_finger_mid_r_02_ctrl.Rotation.Y {
        type=MovieSceneScriptingFloatChannel
        default=0
        thinned_from=35
        keys [[0,-22.3665],[17,-15.6106],[29,40.3956],[40,-50.9226],[51,-40.304],[65,-4.4945],[78,-14.805],[102,-18.6376],[114,0.5371],[125,-30.7273],[136,-27.0919],[150,-14.8318],[166,-20.7556],[171,-20.9858],[190,-18.6376],[202,0.5371],[213,-30.7273],[224,-27.0919],[238,-14.8318],[262,-22.3665],[281,-18.6376],[293,0.5371],[304,-30.7273],[315,-27.0919],[329,-14.8318],[345,-20.7556],[350,-20.9858],[372,-18.6376],[384,0.5371],[395,-30.7273],[406,-27.0919],[420,-14.8318],[446,-22.3665],[450,-22.3665]]
       }
       frnt_finger_mid_r_02_ctrl.Rotation.Z {type=MovieSceneScriptingFloatChannel, default=0, const [0,8]}
       frnt_finger_mid_r_02_ctrl.Scale.X {type=MovieSceneScriptingFloatChannel, default=1, const [1,8]}
       frnt_finger_mid_r_02_ctrl.Scale.Y {type=MovieSceneScriptingFloatChannel, default=1, const [1,8]}
       frnt_finger_mid_r_02_ctrl.Scale.Z {type=MovieSceneScriptingFloatChannel, default=1, const [1,8]}
       frnt_finger_small_r_01_ctrl.Location.X {type=MovieSceneScriptingFloatChannel, default=0, const [0,1]}
       frnt_finger_small_r_01_ctrl.Location.Y {type=MovieSceneScriptingFloatChannel, default=0, const [0,1]}
       frnt_finger_small_r_01_ctrl.Location.Z {type=MovieSceneScriptingFloatChannel, default=0, const [0,1]}
       frnt_finger_small_r_01_ctrl.Rotation.X {type=MovieSceneScriptingFloatChannel, default=0, const [0,1]}
       frnt_finger_small_r_01_ctrl.Rotation.Y {
        type=MovieSceneScriptingFloatChannel
        default=0
        thinned_from=40
        keys [[0,-41.9751],[10,-35.2192],[22,20.787],[33,-48.9991],[44,-38.3805],[58,-24.1031],[74,-38.6114],[79,-41.54],[81,-41.54],[95,-38.8932],[107,-16.9518],[118,-44.2918],[129,-40.1317],[143,-34.5383],[162,-42.6633],[167,-41.54],[183,-38.8932],[195,-16.9518],[206,-44.2918],[217,-40.1317],[231,-34.5383],[255,-43.6894],[258,-41.9751],[260,-41.54],[274,-38.8932],[286,-16.9518],[297,-44.2918],[308,-40.1317],[322,-34.5383],[341,-42.6633],[346,-41.54],[365,-38.8932],[377,-16.9518],[388,-44.2918],[399,-40.1317],[413,-34.5383],[441,-43.6894],[446,-41.9751],[450,-41.9751]]
       }
       frnt_finger_small_r_01_ctrl.Rotation.Z {type=MovieSceneScriptingFloatChannel, default=0, const [0,1]}
       frnt_finger_small_r_01_ctrl.Scale.X {type=MovieSceneScriptingFloatChannel, default=1, const [1,1]}
       frnt_finger_small_r_01_ctrl.Scale.Y {type=MovieSceneScriptingFloatChannel, default=1, const [1,1]}
       frnt_finger_small_r_01_ctrl.Scale.Z {type=MovieSceneScriptingFloatChannel, default=1, const [1,1]}
       frnt_finger_small_r_02_ctrl.Location.X {type=MovieSceneScriptingFloatChannel, default=0, const [0,1]}
       frnt_finger_small_r_02_ctrl.Location.Y {type=MovieSceneScriptingFloatChannel, default=0, const [0,1]}
       frnt_finger_small_r_02_ctrl.Location.Z {type=MovieSceneScriptingFloatChannel, default=0, const [0,1]}
       frnt_finger_small_r_02_ctrl.Rotation.X {type=MovieSceneScriptingFloatChannel, default=0, const [0,1]}
       frnt_finger_small_r_02_ctrl.Rotation.Y {
        type=MovieSceneScriptingFloatChannel
        default=0
        thinned_from=35
        keys [[0,-22.3665],[13,-15.6106],[25,40.3956],[36,-50.9226],[47,-40.304],[61,-4.4945],[74,-14.805],[98,-18.6376],[110,0.5371],[121,-30.7273],[132,-27.0919],[146,-14.8318],[162,-20.7556],[167,-20.9858],[186,-18.6376],[198,0.5371],[209,-30.7273],[220,-27.0919],[234,-14.8318],[258,-22.3665],[277,-18.6376],[289,0.5371],[300,-30.7273],[311,-27.0919],[325,-14.8318],[341,-20.7556],[346,-20.9858],[368,-18.6376],[380,0.5371],[391,-30.7273],[402,-27.0919],[416,-14.8318],[446,-22.3665],[450,-22.3665]]
       }
       frnt_finger_small_r_02_ctrl.Rotation.Z {type=MovieSceneScriptingFloatChannel, default=0, const [0,1]}
       frnt_finger_small_r_02_ctrl.Scale.X {type=MovieSceneScriptingFloatChannel, default=1, const [1,1]}
       frnt_finger_small_r_02_ctrl.Scale.Y {type=MovieSceneScriptingFloatChannel, default=1, const [1,1]}
       frnt_finger_small_r_02_ctrl.Scale.Z {type=MovieSceneScriptingFloatChannel, default=1, const [1,1]}
       back_leg_l_01_ctrl.Location.X {type=MovieSceneScriptingFloatChannel, default=0, const [0,20]}
       back_leg_l_01_ctrl.Location.Y {type=MovieSceneScriptingFloatChannel, default=0, const [0,20]}
       back_leg_l_01_ctrl.Location.Z {type=MovieSceneScriptingFloatChannel, default=0, const [0,20]}
       back_leg_l_01_ctrl.Rotation.X {type=MovieSceneScriptingFloatChannel, default=0, const [0,20]}
       back_leg_l_01_ctrl.Rotation.Y {
        type=MovieSceneScriptingFloatChannel
        default=0
        keys [[0,-12.9695],[5,-14.3233],[17,-26.469],[22,-28.6326],[26,-24.8027],[34,-9.6808],[39,-6.183],[50,-11.1021],[69,-15.2685],[102,-18.5563],[107,-19.5266],[111,-17.809],[119,-11.0274],[124,-9.4588],[135,-11.6648],[154,-13.5333],[176,-12.5023],[178,-13.1094],[190,-18.5563],[195,-19.5266],[199,-17.809],[207,-11.0274],[212,-9.4588],[223,-11.6648],[242,-13.5333],[269,-13.1094],[281,-18.5563],[286,-19.5266],[290,-17.809],[298,-11.0274],[303,-9.4588],[314,-11.6648],[333,-13.5333],[355,-12.5023],[360,-13.1094],[372,-18.5563],[377,-19.5266],[381,-17.809],[389,-11.0274],[394,-9.4588],[405,-11.6648],[424,-13.5333],[450,-12.9695]]
       }
       back_leg_l_01_ctrl.Rotation.Z {type=MovieSceneScriptingFloatChannel, default=0, const [0,20]}
       back_leg_l_01_ctrl.Scale.X {type=MovieSceneScriptingFloatChannel, default=1, const [1,20]}
       back_leg_l_01_ctrl.Scale.Y {type=MovieSceneScriptingFloatChannel, default=1, const [1,20]}
       back_leg_l_01_ctrl.Scale.Z {type=MovieSceneScriptingFloatChannel, default=1, const [1,20]}
       back_leg_l_02_ctrl.Location.X {type=MovieSceneScriptingFloatChannel, default=0, const [0,15]}
       back_leg_l_02_ctrl.Location.Y {type=MovieSceneScriptingFloatChannel, default=0, const [0,15]}
       back_leg_l_02_ctrl.Location.Z {type=MovieSceneScriptingFloatChannel, default=0, const [0,15]}
       back_leg_l_02_ctrl.Rotation.X {type=MovieSceneScriptingFloatChannel, default=0, const [1.8075,8]}
       back_leg_l_02_ctrl.Rotation.Y {
        type=MovieSceneScriptingFloatChannel
        default=0
        keys [[0,7.3489],[8,11.8293],[13,10.1669],[27,2.2925],[39,7.9297],[58,6.858],[78,7.7428],[93,9.2438],[98,8.5664],[112,5.3577],[124,7.6547],[143,7.2181],[176,7.4181],[181,9.2438],[186,8.5664],[200,5.3577],[212,7.6547],[231,7.2181],[264,7.4181],[272,9.2438],[277,8.5664],[291,5.3577],[303,7.6547],[322,7.2181],[355,7.4181],[363,9.2438],[368,8.5664],[382,5.3577],[394,7.6547],[413,7.2181],[450,7.3489]]
       }
       back_leg_l_02_ctrl.Rotation.Z {type=MovieSceneScriptingFloatChannel, default=0, thinned_from=15, keys [[0,-1.1058],[450,-1.1058]]}
       back_leg_l_02_ctrl.Scale.X {type=MovieSceneScriptingFloatChannel, default=1, const [1,15]}
       back_leg_l_02_ctrl.Scale.Y {type=MovieSceneScriptingFloatChannel, default=1, const [1,15]}
       back_leg_l_02_ctrl.Scale.Z {type=MovieSceneScriptingFloatChannel, default=1, const [1,15]}
       bk_finger_small_l_01_ctrl.Location.X {type=MovieSceneScriptingFloatChannel, default=0, const [0,5]}
       bk_finger_small_l_01_ctrl.Location.Y {type=MovieSceneScriptingFloatChannel, default=0, const [0,5]}
       bk_finger_small_l_01_ctrl.Location.Z {type=MovieSceneScriptingFloatChannel, default=0, const [0,5]}
       bk_finger_small_l_01_ctrl.Rotation.X {type=MovieSceneScriptingFloatChannel, default=0, const [0,5]}
       bk_finger_small_l_01_ctrl.Rotation.Y {
        type=MovieSceneScriptingFloatChannel
        default=0
        thinned_from=25
        keys [[0,0],[14,9.1895],[29,-45.5915],[42,-12.0663],[57,-26.2379],[78,3.7894],[99,5.1034],[114,-20.2481],[127,-4.7333],[142,-11.2916],[187,5.1034],[202,-20.2481],[215,-4.7333],[230,-11.2916],[278,5.1034],[293,-20.2481],[306,-4.7333],[321,-11.2916],[369,5.1034],[384,-20.2481],[397,-4.7333],[412,-11.2916],[446,0],[450,0]]
       }
       bk_finger_small_l_01_ctrl.Rotation.Z {type=MovieSceneScriptingFloatChannel, default=0, const [0,5]}
       bk_finger_small_l_01_ctrl.Scale.X {type=MovieSceneScriptingFloatChannel, default=1, const [1,5]}
       bk_finger_small_l_01_ctrl.Scale.Y {type=MovieSceneScriptingFloatChannel, default=1, const [1,5]}
       bk_finger_small_l_01_ctrl.Scale.Z {type=MovieSceneScriptingFloatChannel, default=1, const [1,5]}
       bk_finger_small_l_02_ctrl.Location.X {type=MovieSceneScriptingFloatChannel, default=0, thinned_from=4, keys [[0,0],[450,0]]}
       bk_finger_small_l_02_ctrl.Location.Y {type=MovieSceneScriptingFloatChannel, default=0, thinned_from=4, keys [[0,0],[450,0]]}
       bk_finger_small_l_02_ctrl.Location.Z {type=MovieSceneScriptingFloatChannel, default=0, thinned_from=4, keys [[0,0],[450,0]]}
       bk_finger_small_l_02_ctrl.Rotation.X {type=MovieSceneScriptingFloatChannel, default=0, const [-2.2135,4]}
       bk_finger_small_l_02_ctrl.Rotation.Y {
        type=MovieSceneScriptingFloatChannel
        default=0
        thinned_from=29
        keys [[0,-38.9068],[16,24.3426],[31,-55.1488],[44,5.8676],[59,-39.9199],[78,-34.3962],[85,-38.8023],[101,-11.8752],[116,-45.717],[129,-19.7405],[144,-39.2336],[171,-38.8023],[189,-11.8752],[204,-45.717],[217,-19.7405],[232,-39.2336],[264,-38.8023],[280,-11.8752],[295,-45.717],[308,-19.7405],[323,-39.2336],[350,-38.8023],[371,-11.8752],[386,-45.717],[399,-19.7405],[414,-39.2336],[450,-38.9068]]
       }
       bk_finger_small_l_02_ctrl.Rotation.Z {type=MovieSceneScriptingFloatChannel, default=0, const [-16.5201,4]}
       bk_finger_small_l_02_ctrl.Scale.X {type=MovieSceneScriptingFloatChannel, default=1, const [1,4]}
       bk_finger_small_l_02_ctrl.Scale.Y {type=MovieSceneScriptingFloatChannel, default=1, const [1,4]}
       bk_finger_small_l_02_ctrl.Scale.Z {type=MovieSceneScriptingFloatChannel, default=1, const [1,4]}
       bk_finger_mid_l_01_ctrl.Location.X {type=MovieSceneScriptingFloatChannel, default=0, thinned_from=7, keys [[0,0],[78,0.1565],[171,0.471],[350,0.471],[446,0],[450,0]]}
       bk_finger_mid_l_01_ctrl.Location.Y {type=MovieSceneScriptingFloatChannel, default=0, thinned_from=7, keys [[0,0],[78,0.1565],[171,0.471],[350,0.471],[446,0],[450,0]]}
       bk_finger_mid_l_01_ctrl.Location.Z {type=MovieSceneScriptingFloatChannel, default=0, thinned_from=7, keys [[0,0],[78,0.1565],[171,0.471],[350,0.471],[446,0],[450,0]]}
       bk_finger_mid_l_01_ctrl.Rotation.X {type=MovieSceneScriptingFloatChannel, default=0, thinned_from=7, keys [[0,0],[78,0.1565],[171,0.471],[350,0.471],[446,0],[450,0]]}
       bk_finger_mid_l_01_ctrl.Rotation.Y {
        type=MovieSceneScriptingFloatChannel
        default=0
        thinned_from=27
        keys [[0,0],[17,18.9555],[32,-54.5307],[45,-22.4084],[60,-37.9829],[78,-3.4845],[102,8.4239],[117,-22.4076],[130,-8.9305],[145,-15.4649],[171,0.471],[190,8.4239],[205,-22.4076],[218,-8.9305],[233,-15.4649],[281,8.4239],[296,-22.4076],[309,-8.9305],[324,-15.4649],[350,0.471],[372,8.4239],[387,-22.4076],[400,-8.9305],[415,-15.4649],[446,0],[450,0]]
       }
       bk_finger_mid_l_01_ctrl.Rotation.Z {type=MovieSceneScriptingFloatChannel, default=0, const [0,5]}
       bk_finger_mid_l_01_ctrl.Scale.X {type=MovieSceneScriptingFloatChannel, default=1, const [1,5]}
       bk_finger_mid_l_01_ctrl.Scale.Y {type=MovieSceneScriptingFloatChannel, default=1, const [1,5]}
       bk_finger_mid_l_01_ctrl.Scale.Z {type=MovieSceneScriptingFloatChannel, default=1, const [1,5]}
       bk_finger_mid_l_02_ctrl.Location.X {type=MovieSceneScriptingFloatChannel, default=0, thinned_from=4, keys [[0,0],[450,0]]}
       bk_finger_mid_l_02_ctrl.Location.Y {type=MovieSceneScriptingFloatChannel, default=0, thinned_from=4, keys [[0,0],[450,0]]}
       bk_finger_mid_l_02_ctrl.Location.Z {type=MovieSceneScriptingFloatChannel, default=0, thinned_from=4, keys [[0,0],[450,0]]}
       bk_finger_mid_l_02_ctrl.Rotation.X {type=MovieSceneScriptingFloatChannel, default=0, const [-4.0253,4]}
       bk_finger_mid_l_02_ctrl.Rotation.Y {
        type=MovieSceneScriptingFloatChannel
        default=0
        thinned_from=27
        keys [[0,-8.865],[19,54.3844],[34,-25.107],[47,35.9094],[62,-9.8781],[78,-0.2476],[104,17.2936],[119,-14.3089],[132,9.9487],[147,-8.2545],[171,-7.8518],[192,17.2936],[207,-14.3089],[220,9.9487],[235,-8.2545],[283,17.2936],[298,-14.3089],[311,9.9487],[326,-8.2545],[350,-7.8518],[374,17.2936],[389,-14.3089],[402,9.9487],[417,-8.2545],[446,-8.865],[450,-8.865]]
       }
       bk_finger_mid_l_02_ctrl.Rotation.Z {type=MovieSceneScriptingFloatChannel, default=0, const [-2.1956,4]}
       bk_finger_mid_l_02_ctrl.Scale.X {type=MovieSceneScriptingFloatChannel, default=1, const [1,4]}
       bk_finger_mid_l_02_ctrl.Scale.Y {type=MovieSceneScriptingFloatChannel, default=1, const [1,4]}
       bk_finger_mid_l_02_ctrl.Scale.Z {type=MovieSceneScriptingFloatChannel, default=1, const [1,4]}
       bk_finger_big_l_01_ctrl.Location.X {type=MovieSceneScriptingFloatChannel, default=0, thinned_from=10, keys [[0,0],[450,0]]}
       bk_finger_big_l_01_ctrl.Location.Y {type=MovieSceneScriptingFloatChannel, default=0, thinned_from=10, keys [[0,0],[450,0]]}
       bk_finger_big_l_01_ctrl.Location.Z {type=MovieSceneScriptingFloatChannel, default=0, thinned_from=10, keys [[0,0],[450,0]]}
       bk_finger_big_l_01_ctrl.Rotation.X {type=MovieSceneScriptingFloatChannel, default=0, const [-4.1353,10]}
       bk_finger_big_l_01_ctrl.Rotation.Y {
        type=MovieSceneScriptingFloatChannel
        default=0
        thinned_from=27
        keys [[0,-38.9068],[20,15.9912],[35,-101.0813],[48,-52.592],[63,-66.7636],[78,-46.0951],[105,-9.9572],[120,-72.0881],[133,-46.3546],[148,-53.8755],[171,-39.0918],[193,-9.9572],[208,-72.0881],[221,-46.3546],[236,-53.8755],[284,-9.9572],[299,-72.0881],[312,-46.3546],[327,-53.8755],[350,-39.0918],[375,-9.9572],[390,-72.0881],[403,-46.3546],[418,-53.8755],[446,-38.9068],[450,-38.9068]]
       }
       bk_finger_big_l_01_ctrl.Rotation.Z {type=MovieSceneScriptingFloatChannel, default=0, const [8.0558,10]}
       bk_finger_big_l_01_ctrl.Scale.X {type=MovieSceneScriptingFloatChannel, default=1, const [1,10]}
       bk_finger_big_l_01_ctrl.Scale.Y {type=MovieSceneScriptingFloatChannel, default=1, const [1,10]}
       bk_finger_big_l_01_ctrl.Scale.Z {type=MovieSceneScriptingFloatChannel, default=1, const [1,10]}
       bk_finger_big_l_02_ctrl.Location.X {type=MovieSceneScriptingFloatChannel, default=0, thinned_from=4, keys [[0,0],[450,0]]}
       bk_finger_big_l_02_ctrl.Location.Y {type=MovieSceneScriptingFloatChannel, default=0, thinned_from=4, keys [[0,0],[450,0]]}
       bk_finger_big_l_02_ctrl.Location.Z {type=MovieSceneScriptingFloatChannel, default=0, thinned_from=4, keys [[0,0],[450,0]]}
       bk_finger_big_l_02_ctrl.Rotation.X {type=MovieSceneScriptingFloatChannel, default=0, const [-1.8649,4]}
       bk_finger_big_l_02_ctrl.Rotation.Y {
        type=MovieSceneScriptingFloatChannel
        default=0
        thinned_from=27
        keys [[0,-8.865],[22,21.0267],[37,-25.107],[50,27.0377],[65,-9.8781],[78,-6.0989],[107,4.6854],[122,-15.5248],[135,7.3187],[150,-8.8534],[171,-8.4095],[195,4.6854],[210,-15.5248],[223,7.3187],[238,-8.8534],[286,4.6854],[301,-15.5248],[314,7.3187],[329,-8.8534],[350,-8.4095],[377,4.6854],[392,-15.5248],[405,7.3187],[420,-8.8534],[450,-8.865]]
       }
       bk_finger_big_l_02_ctrl.Rotation.Z {type=MovieSceneScriptingFloatChannel, default=0, const [4.6264,4]}
       bk_finger_big_l_02_ctrl.Scale.X {type=MovieSceneScriptingFloatChannel, default=1, const [1,4]}
       bk_finger_big_l_02_ctrl.Scale.Y {type=MovieSceneScriptingFloatChannel, default=1, const [1,4]}
       bk_finger_big_l_02_ctrl.Scale.Z {type=MovieSceneScriptingFloatChannel, default=1, const [1,4]}
       back_leg_r_01_ctrl.Location.X {type=MovieSceneScriptingFloatChannel, default=0, const [0,20]}
       back_leg_r_01_ctrl.Location.Y {type=MovieSceneScriptingFloatChannel, default=0, const [0,20]}
       back_leg_r_01_ctrl.Location.Z {type=MovieSceneScriptingFloatChannel, default=0, const [0,20]}
       back_leg_r_01_ctrl.Rotation.X {type=MovieSceneScriptingFloatChannel, default=0, const [0,20]}
       back_leg_r_01_ctrl.Rotation.Y {type=MovieSceneScriptingFloatChannel, default=0, const [0,15]}
       back_leg_r_01_ctrl.Rotation.Z {
        type=MovieSceneScriptingFloatChannel
        default=0
        keys [[0,-12.9695],[11,-14.3233],[23,-26.469],[28,-28.6326],[32,-24.8027],[40,-9.6808],[45,-6.183],[56,-11.1021],[75,-15.2685],[108,-18.5563],[113,-19.5266],[117,-17.809],[125,-11.0274],[130,-9.4588],[141,-11.6648],[160,-13.5333],[176,-12.5023],[184,-13.1094],[196,-18.5563],[201,-19.5266],[205,-17.809],[213,-11.0274],[218,-9.4588],[229,-11.6648],[248,-13.5333],[275,-13.1094],[287,-18.5563],[292,-19.5266],[296,-17.809],[304,-11.0274],[309,-9.4588],[320,-11.6648],[339,-13.5333],[355,-12.5023],[366,-13.1094],[378,-18.5563],[383,-19.5266],[387,-17.809],[395,-11.0274],[400,-9.4588],[411,-11.6648],[430,-13.5333],[450,-12.9695]]
       }
       back_leg_r_01_ctrl.Scale.X {type=MovieSceneScriptingFloatChannel, default=1, const [1,20]}
       back_leg_r_01_ctrl.Scale.Y {type=MovieSceneScriptingFloatChannel, default=1, const [1,20]}
       back_leg_r_01_ctrl.Scale.Z {type=MovieSceneScriptingFloatChannel, default=1, const [1,20]}
       back_leg_r_02_ctrl.Location.X {type=MovieSceneScriptingFloatChannel, default=0, const [0,15]}
       back_leg_r_02_ctrl.Location.Y {type=MovieSceneScriptingFloatChannel, default=0, const [0,15]}
       back_leg_r_02_ctrl.Location.Z {type=MovieSceneScriptingFloatChannel, default=0, const [0,15]}
       back_leg_r_02_ctrl.Rotation.X {type=MovieSceneScriptingFloatChannel, default=0, const [0,15]}
       back_leg_r_02_ctrl.Rotation.Y {type=MovieSceneScriptingFloatChannel, default=0, const [0,15]}
       back_leg_r_02_ctrl.Rotation.Z {
        type=MovieSceneScriptingFloatChannel
        default=0
        keys [[0,7.3489],[14,11.8293],[19,10.1669],[33,2.2925],[45,7.9297],[64,6.858],[78,7.5087],[99,9.2438],[104,8.5664],[118,5.3577],[130,7.6547],[149,7.2181],[176,7.4181],[187,9.2438],[192,8.5664],[206,5.3577],[218,7.6547],[237,7.2181],[264,7.4181],[278,9.2438],[283,8.5664],[297,5.3577],[309,7.6547],[328,7.2181],[355,7.4181],[369,9.2438],[374,8.5664],[388,5.3577],[400,7.6547],[419,7.2181],[450,7.3489]]
       }
       back_leg_r_02_ctrl.Scale.X {type=MovieSceneScriptingFloatChannel, default=1, const [1,15]}
       back_leg_r_02_ctrl.Scale.Y {type=MovieSceneScriptingFloatChannel, default=1, const [1,15]}
       back_leg_r_02_ctrl.Scale.Z {type=MovieSceneScriptingFloatChannel, default=1, const [1,15]}
       bk_finger_big_r_01_ctrl.Location.X {type=MovieSceneScriptingFloatChannel, default=0, const [0,4]}
       bk_finger_big_r_01_ctrl.Location.Y {type=MovieSceneScriptingFloatChannel, default=0, const [0,4]}
       bk_finger_big_r_01_ctrl.Location.Z {type=MovieSceneScriptingFloatChannel, default=0, const [0,4]}
       bk_finger_big_r_01_ctrl.Rotation.X {type=MovieSceneScriptingFloatChannel, default=0, const [0,4]}
       bk_finger_big_r_01_ctrl.Rotation.Y {
        type=MovieSceneScriptingFloatChannel
        default=0
        thinned_from=27
        keys [[0,-38.9068],[26,15.9912],[41,-101.0813],[54,-52.592],[69,-66.7636],[78,-54.7081],[111,-8.944],[126,-72.5207],[139,-46.1884],[154,-53.8843],[171,-38.7566],[199,-8.944],[214,-72.5207],[227,-46.1884],[242,-53.8843],[290,-8.944],[305,-72.5207],[318,-46.1884],[333,-53.8843],[350,-38.7566],[381,-8.944],[396,-72.5207],[409,-46.1884],[424,-53.8843],[446,-38.9068],[450,-38.9068]]
       }
       bk_finger_big_r_01_ctrl.Rotation.Z {type=MovieSceneScriptingFloatChannel, default=0, const [0,4]}
       bk_finger_big_r_01_ctrl.Scale.X {type=MovieSceneScriptingFloatChannel, default=1, const [1,4]}
       bk_finger_big_r_01_ctrl.Scale.Y {type=MovieSceneScriptingFloatChannel, default=1, const [1,4]}
       bk_finger_big_r_01_ctrl.Scale.Z {type=MovieSceneScriptingFloatChannel, default=1, const [1,4]}
       bk_finger_big_r_02_ctrl.Location.X {type=MovieSceneScriptingFloatChannel, default=0, const [0,4]}
       bk_finger_big_r_02_ctrl.Location.Y {type=MovieSceneScriptingFloatChannel, default=0, const [0,4]}
       bk_finger_big_r_02_ctrl.Location.Z {type=MovieSceneScriptingFloatChannel, default=0, const [0,4]}
       bk_finger_big_r_02_ctrl.Rotation.X {type=MovieSceneScriptingFloatChannel, default=0, const [0,4]}
       bk_finger_big_r_02_ctrl.Rotation.Y {
        type=MovieSceneScriptingFloatChannel
        default=0
        thinned_from=27
        keys [[0,-8.865],[28,21.0267],[43,-25.107],[56,27.0377],[71,-8.9541],[78,-7.4567],[113,4.7755],[128,-15.7196],[141,7.4459],[156,-8.9541],[171,-8.504],[201,4.7755],[216,-15.7196],[229,7.4459],[244,-8.9541],[292,4.7755],[307,-15.7196],[320,7.4459],[335,-8.9541],[350,-8.504],[383,4.7755],[398,-15.7196],[411,7.4459],[426,-8.9541],[450,-8.865]]
       }
       bk_finger_big_r_02_ctrl.Rotation.Z {type=MovieSceneScriptingFloatChannel, default=0, const [0,4]}
       bk_finger_big_r_02_ctrl.Scale.X {type=MovieSceneScriptingFloatChannel, default=1, const [1,4]}
       bk_finger_big_r_02_ctrl.Scale.Y {type=MovieSceneScriptingFloatChannel, default=1, const [1,4]}
       bk_finger_big_r_02_ctrl.Scale.Z {type=MovieSceneScriptingFloatChannel, default=1, const [1,4]}
       bk_finger_mid_r_01_ctrl.Location.X {type=MovieSceneScriptingFloatChannel, default=0, const [0,4]}
       bk_finger_mid_r_01_ctrl.Location.Y {type=MovieSceneScriptingFloatChannel, default=0, const [0,4]}
       bk_finger_mid_r_01_ctrl.Location.Z {type=MovieSceneScriptingFloatChannel, default=0, const [0,4]}
       bk_finger_mid_r_01_ctrl.Rotation.X {type=MovieSceneScriptingFloatChannel, default=0, const [0,4]}
       bk_finger_mid_r_01_ctrl.Rotation.Y {
        type=MovieSceneScriptingFloatChannel
        default=0
        keys [[0,0],[23,18.9555],[38,-54.5307],[51,-22.4084],[66,-37.9829],[78,-13.7509],[108,10.0492],[123,-26.9082],[136,-10.7534],[151,-18.5861],[196,10.0492],[211,-26.9082],[224,-10.7534],[239,-18.5861],[287,10.0492],[302,-26.9082],[315,-10.7534],[330,-18.5861],[378,10.0492],[393,-26.9082],[406,-10.7534],[421,-18.5861],[448,0.3089],[450,0]]
       }
       bk_finger_mid_r_01_ctrl.Rotation.Z {type=MovieSceneScriptingFloatChannel, default=0, const [0,4]}
       bk_finger_mid_r_01_ctrl.Scale.X {type=MovieSceneScriptingFloatChannel, default=1, const [1,4]}
       bk_finger_mid_r_01_ctrl.Scale.Y {type=MovieSceneScriptingFloatChannel, default=1, const [1,4]}
       bk_finger_mid_r_01_ctrl.Scale.Z {type=MovieSceneScriptingFloatChannel, default=1, const [1,4]}
       bk_finger_mid_r_02_ctrl.Location.X {type=MovieSceneScriptingFloatChannel, default=0, const [0,4]}
       bk_finger_mid_r_02_ctrl.Location.Y {type=MovieSceneScriptingFloatChannel, default=0, const [0,4]}
       bk_finger_mid_r_02_ctrl.Location.Z {type=MovieSceneScriptingFloatChannel, default=0, const [0,4]}
       bk_finger_mid_r_02_ctrl.Rotation.X {type=MovieSceneScriptingFloatChannel, default=0, const [0,4]}
       bk_finger_mid_r_02_ctrl.Rotation.Y {
        type=MovieSceneScriptingFloatChannel
        default=0
        thinned_from=30
        keys [[0,-8.865],[25,54.3844],[40,-25.107],[53,35.9094],[68,-9.8781],[78,-6.518],[83,-8.5223],[85,-8.5223],[110,22.5058],[125,-16.4901],[138,13.4425],[153,-9.0193],[171,-8.5223],[198,22.5058],[213,-16.4901],[226,13.4425],[241,-9.0193],[264,-8.5223],[289,22.5058],[304,-16.4901],[317,13.4425],[332,-9.0193],[350,-8.5223],[380,22.5058],[395,-16.4901],[408,13.4425],[423,-9.0193],[446,-8.865],[450,-8.865]]
       }
       bk_finger_mid_r_02_ctrl.Rotation.Z {type=MovieSceneScriptingFloatChannel, default=0, const [0,4]}
       bk_finger_mid_r_02_ctrl.Scale.X {type=MovieSceneScriptingFloatChannel, default=1, const [1,4]}
       bk_finger_mid_r_02_ctrl.Scale.Y {type=MovieSceneScriptingFloatChannel, default=1, const [1,4]}
       bk_finger_mid_r_02_ctrl.Scale.Z {type=MovieSceneScriptingFloatChannel, default=1, const [1,4]}
       bk_finger_small_r_01_ctrl.Location.X {type=MovieSceneScriptingFloatChannel, default=0, const [0,10]}
       bk_finger_small_r_01_ctrl.Location.Y {type=MovieSceneScriptingFloatChannel, default=0, const [0,10]}
       bk_finger_small_r_01_ctrl.Location.Z {type=MovieSceneScriptingFloatChannel, default=0, const [0,10]}
       bk_finger_small_r_01_ctrl.Rotation.X {type=MovieSceneScriptingFloatChannel, default=0, const [0,10]}
       bk_finger_small_r_01_ctrl.Rotation.Y {
        type=MovieSceneScriptingFloatChannel
        default=0
        thinned_from=29
        keys [[0,0],[20,9.1895],[35,-45.5915],[48,-12.0663],[63,-26.2379],[78,-3.6016],[85,-0.0646],[105,3.6207],[120,-18.3481],[133,-4.9035],[148,-10.5867],[171,-0.0646],[193,3.6207],[208,-18.3481],[221,-4.9035],[236,-10.5867],[264,-0.0646],[284,3.6207],[299,-18.3481],[312,-4.9035],[327,-10.5867],[350,-0.0646],[375,3.6207],[390,-18.3481],[403,-4.9035],[418,-10.5867],[446,0],[450,0]]
       }
       bk_finger_small_r_01_ctrl.Rotation.Z {type=MovieSceneScriptingFloatChannel, default=0, const [0,10]}
       bk_finger_small_r_01_ctrl.Scale.X {type=MovieSceneScriptingFloatChannel, default=1, const [1,10]}
       bk_finger_small_r_01_ctrl.Scale.Y {type=MovieSceneScriptingFloatChannel, default=1, const [1,10]}
       bk_finger_small_r_01_ctrl.Scale.Z {type=MovieSceneScriptingFloatChannel, default=1, const [1,10]}
       bk_finger_small_r_02_ctrl.Location.X {type=MovieSceneScriptingFloatChannel, default=0, const [0,10]}
       bk_finger_small_r_02_ctrl.Location.Y {type=MovieSceneScriptingFloatChannel, default=0, const [0,10]}
       bk_finger_small_r_02_ctrl.Location.Z {type=MovieSceneScriptingFloatChannel, default=0, const [0,10]}
       bk_finger_small_r_02_ctrl.Rotation.X {type=MovieSceneScriptingFloatChannel, default=0, const [0,10]}
       bk_finger_small_r_02_ctrl.Rotation.Y {
        type=MovieSceneScriptingFloatChannel
        default=0
        thinned_from=29
        keys [[0,-38.9068],[22,24.3426],[37,-55.1488],[50,5.8676],[65,-39.9199],[78,-35.8647],[85,-38.2495],[107,-12.4071],[122,-45.7581],[135,-20.1583],[150,-39.3687],[171,-38.9436],[195,-12.4071],[210,-45.7581],[223,-20.1583],[238,-39.3687],[264,-38.2495],[286,-12.4071],[301,-45.7581],[314,-20.1583],[329,-39.3687],[350,-38.9436],[377,-12.4071],[392,-45.7581],[405,-20.1583],[420,-39.3687],[450,-38.9068]]
       }
       bk_finger_small_r_02_ctrl.Rotation.Z {type=MovieSceneScriptingFloatChannel, default=0, const [0,10]}
       bk_finger_small_r_02_ctrl.Scale.X {type=MovieSceneScriptingFloatChannel, default=1, const [1,10]}
       bk_finger_small_r_02_ctrl.Scale.Y {type=MovieSceneScriptingFloatChannel, default=1, const [1,10]}
       bk_finger_small_r_02_ctrl.Scale.Z {type=MovieSceneScriptingFloatChannel, default=1, const [1,10]}
       dyn_tail_base_ctrl.Location.X {type=MovieSceneScriptingFloatChannel, default=0, const [0,6]}
       dyn_tail_base_ctrl.Location.Y {type=MovieSceneScriptingFloatChannel, default=0, const [0,6]}
       dyn_tail_base_ctrl.Location.Z {type=MovieSceneScriptingFloatChannel, default=0, const [0,6]}
       dyn_tail_base_ctrl.Rotation.X {type=MovieSceneScriptingFloatChannel, default=0, const [0,6]}
       dyn_tail_base_ctrl.Rotation.Y {
        type=MovieSceneScriptingFloatChannel
        default=0
        keys [[0,0],[7,-1.5722],[22,7.5309],[32,9.3984],[53,-9.846],[71,-3.5903],[78,-3.701],[92,0.0727],[107,4.4993],[117,5.4074],[138,-3.9506],[156,-0.9086],[166,-2.5496],[180,0.0727],[195,4.4993],[205,5.4074],[226,-3.9506],[244,-0.9086],[271,0.0727],[286,4.4993],[296,5.4074],[317,-3.9506],[335,-0.9086],[345,-2.5496],[362,0.0727],[377,4.4993],[387,5.4074],[408,-3.9506],[426,-0.9086],[448,-0.0179],[450,0]]
       }
       dyn_tail_base_ctrl.Rotation.Z {type=MovieSceneScriptingFloatChannel, default=0, const [0,6]}
       dyn_tail_base_ctrl.Scale.X {type=MovieSceneScriptingFloatChannel, default=1, const [1,6]}
       dyn_tail_base_ctrl.Scale.Y {type=MovieSceneScriptingFloatChannel, default=1, const [1,6]}
       dyn_tail_base_ctrl.Scale.Z {type=MovieSceneScriptingFloatChannel, default=1, const [1,6]}
       dyn_tail_01_ctrl.Location.X {type=MovieSceneScriptingFloatChannel, default=0, const [0,15]}
       dyn_tail_01_ctrl.Location.Y {type=MovieSceneScriptingFloatChannel, default=0, const [0,15]}
       dyn_tail_01_ctrl.Location.Z {type=MovieSceneScriptingFloatChannel, default=0, const [0,15]}
       dyn_tail_01_ctrl.Rotation.X {type=MovieSceneScriptingFloatChannel, default=0, const [0,15]}
       dyn_tail_01_ctrl.Rotation.Y {type=MovieSceneScriptingFloatChannel, default=0, const [0,15]}
       dyn_tail_01_ctrl.Rotation.Z {type=MovieSceneScriptingFloatChannel, default=0, const [0,15]}
       dyn_tail_01_ctrl.Scale.X {type=MovieSceneScriptingFloatChannel, default=1, const [1,15]}
       dyn_tail_01_ctrl.Scale.Y {type=MovieSceneScriptingFloatChannel, default=1, const [1,15]}
       dyn_tail_01_ctrl.Scale.Z {type=MovieSceneScriptingFloatChannel, default=1, const [1,15]}
       dyn_tail_02_ctrl.Location.X {type=MovieSceneScriptingFloatChannel, default=0, const [0,5]}
       dyn_tail_02_ctrl.Location.Y {type=MovieSceneScriptingFloatChannel, default=0, const [0,5]}
       dyn_tail_02_ctrl.Location.Z {type=MovieSceneScriptingFloatChannel, default=0, const [0,5]}
       dyn_tail_02_ctrl.Rotation.X {type=MovieSceneScriptingFloatChannel, default=0, const [0,5]}
       dyn_tail_02_ctrl.Rotation.Y {
        type=MovieSceneScriptingFloatChannel
        default=0
        thinned_from=49
        keys [[0,0],[3,-1.584],[16,-12.8902],[27,11.0488],[36,21.1063],[44,12.764],[59,-10.2228],[76,-3.0599],[78,-2.6272],[83,0.4097],[85,0.4097],[88,-0.0002],[101,-2.9255],[112,3.2684],[121,5.8707],[129,3.7122],[144,-2.2354],[161,-0.382],[166,-0.1382],[171,0.4097],[176,-0.0002],[189,-2.9255],[200,3.2684],[209,5.8707],[217,3.7122],[232,-2.2354],[249,-0.382],[262,0],[264,0.4097],[267,-0.0002],[280,-2.9255],[291,3.2684],[300,5.8707],[308,3.7122],[323,-2.2354],[340,-0.382],[345,-0.1382],[350,0.4097],[355,0.4097],[358,-0.0002],[371,-2.9255],[382,3.2684],[391,5.8707],[399,3.7122],[414,-2.2354],[431,-0.382],[446,0],[450,0]]
       }
       dyn_tail_02_ctrl.Rotation.Z {type=MovieSceneScriptingFloatChannel, default=0, const [0,5]}
       dyn_tail_02_ctrl.Scale.X {type=MovieSceneScriptingFloatChannel, default=1, const [1,6]}
       dyn_tail_02_ctrl.Scale.Y {type=MovieSceneScriptingFloatChannel, default=1, const [1,6]}
       dyn_tail_02_ctrl.Scale.Z {type=MovieSceneScriptingFloatChannel, default=1, const [1,6]}
       dyn_tail_03_ctrl.Location.X {type=MovieSceneScriptingFloatChannel, default=0, const [0,5]}
       dyn_tail_03_ctrl.Location.Y {type=MovieSceneScriptingFloatChannel, default=0, const [0,5]}
       dyn_tail_03_ctrl.Location.Z {type=MovieSceneScriptingFloatChannel, default=0, const [0,5]}
       dyn_tail_03_ctrl.Rotation.X {type=MovieSceneScriptingFloatChannel, default=0, const [0,5]}
       dyn_tail_03_ctrl.Rotation.Y {
        type=MovieSceneScriptingFloatChannel
        default=0
        thinned_from=49
        keys [[0,0],[6,-4.0184],[19,-36.7414],[29,-1.3937],[39,12.9917],[47,4.6494],[62,-25.0996],[78,-3.5263],[83,0.8372],[85,0.8372],[91,-1.1168],[104,-17.029],[114,0.1595],[124,7.1547],[132,3.0981],[147,-11.368],[164,-0.6507],[166,-0.4388],[171,0.8372],[176,0.8372],[179,-1.1168],[192,-17.029],[202,0.1595],[212,7.1547],[220,3.0981],[235,-11.368],[252,-0.6507],[262,0],[264,0.8372],[270,-1.1168],[283,-17.029],[293,0.1595],[303,7.1547],[311,3.0981],[326,-11.368],[343,-0.6507],[345,-0.4388],[350,0.8372],[355,0.8372],[361,-1.1168],[374,-17.029],[384,0.1595],[394,7.1547],[402,3.0981],[417,-11.368],[434,-0.6507],[446,0],[450,0]]
       }
       dyn_tail_03_ctrl.Rotation.Z {type=MovieSceneScriptingFloatChannel, default=0, const [0,5]}
       dyn_tail_03_ctrl.Scale.X {type=MovieSceneScriptingFloatChannel, default=1, const [1,6]}
       dyn_tail_03_ctrl.Scale.Y {type=MovieSceneScriptingFloatChannel, default=1, const [1,6]}
       dyn_tail_03_ctrl.Scale.Z {type=MovieSceneScriptingFloatChannel, default=1, const [1,6]}
       dyn_tail_04_ctrl.Location.X {type=MovieSceneScriptingFloatChannel, default=0, const [0,5]}
       dyn_tail_04_ctrl.Location.Y {type=MovieSceneScriptingFloatChannel, default=0, const [0,5]}
       dyn_tail_04_ctrl.Location.Z {type=MovieSceneScriptingFloatChannel, default=0, const [0,5]}
       dyn_tail_04_ctrl.Rotation.X {type=MovieSceneScriptingFloatChannel, default=0, const [0,5]}
       dyn_tail_04_ctrl.Rotation.Y {
        type=MovieSceneScriptingFloatChannel
        default=0
        thinned_from=47
        keys [[0,0],[10,-4.0184],[23,-36.7414],[33,-1.3937],[43,12.9917],[51,4.6494],[66,-25.0996],[78,-8.5147],[83,0.8372],[85,0.8372],[95,-1.1168],[108,-17.029],[118,0.1595],[128,7.1547],[136,3.0981],[151,-11.368],[166,-1.2652],[171,0.8372],[176,0.8372],[183,-1.1168],[196,-17.029],[206,0.1595],[216,7.1547],[224,3.0981],[239,-11.368],[256,-0.6507],[262,0],[264,0.8372],[274,-1.1168],[287,-17.029],[297,0.1595],[307,7.1547],[315,3.0981],[330,-11.368],[345,-1.2652],[350,0.8372],[355,0.8372],[365,-1.1168],[378,-17.029],[388,0.1595],[398,7.1547],[406,3.0981],[421,-11.368],[438,-0.6507],[446,0],[450,0]]
       }
       dyn_tail_04_ctrl.Rotation.Z {type=MovieSceneScriptingFloatChannel, default=0, const [0,5]}
       dyn_tail_04_ctrl.Scale.X {type=MovieSceneScriptingFloatChannel, default=1, const [1,6]}
       dyn_tail_04_ctrl.Scale.Y {type=MovieSceneScriptingFloatChannel, default=1, const [1,6]}
       dyn_tail_04_ctrl.Scale.Z {type=MovieSceneScriptingFloatChannel, default=1, const [1,6]}
       llama_breathing_jnt_ctrl.Location.X {
        type=MovieSceneScriptingFloatChannel
        default=0
        keys [[0,1.3799],[5,1.1655],[10,-1.6796],[18,-2.1512],[25,-1.6881],[39,1.463],[49,1.8268],[74,1.6938],[95,-1.6796],[103,-2.1512],[110,-1.6881],[134,1.8268],[162,1.6632],[183,-1.6796],[191,-2.1512],[198,-1.6881],[222,1.8268],[240,1.6938],[274,-1.6796],[282,-2.1512],[289,-1.6881],[313,1.8268],[341,1.6632],[365,-1.6796],[373,-2.1512],[380,-1.6881],[404,1.8268],[422,1.6938],[448,1.3842],[450,1.3799]]
       }
       llama_breathing_jnt_ctrl.Location.Y {type=MovieSceneScriptingFloatChannel, default=0, const [0,18]}
       llama_breathing_jnt_ctrl.Location.Z {type=MovieSceneScriptingFloatChannel, default=0, const [0,18]}
       llama_breathing_jnt_ctrl.Rotation.X {type=MovieSceneScriptingFloatChannel, default=0, const [0,18]}
       llama_breathing_jnt_ctrl.Rotation.Y {type=MovieSceneScriptingFloatChannel, default=0, const [0,18]}
       llama_breathing_jnt_ctrl.Rotation.Z {type=MovieSceneScriptingFloatChannel, default=0, const [0,18]}
       llama_breathing_jnt_ctrl.Scale.X {type=MovieSceneScriptingFloatChannel, default=1, const [1,18]}
       llama_breathing_jnt_ctrl.Scale.Y {
        type=MovieSceneScriptingFloatChannel
        default=1
        keys [[0,1.4155],[15,2.1443],[23,2.2708],[30,2.1505],[44,1.3562],[54,1.09],[63,1.1455],[78,1.1767],[100,1.9215],[108,2.048],[115,1.9277],[139,1.09],[149,1.149],[166,1.1957],[188,1.9215],[196,2.048],[203,1.9277],[227,1.09],[234,1.149],[245,1.1822],[279,1.9215],[287,2.048],[294,1.9277],[318,1.09],[328,1.149],[343,1.1822],[370,1.9215],[378,2.048],[385,1.9277],[409,1.09],[415,1.1491],[427,1.1781],[448,1.4108],[450,1.4155]]
       }
       llama_breathing_jnt_ctrl.Scale.Z {type=MovieSceneScriptingFloatChannel, default=1, const [1,18]}
       llama_neck_02_ctrl.Space {
        type=MovieSceneScriptingSpaceChannel
        const [
         "<Struct 'MovieSceneControlRigSpaceBaseKey' (0x49d0ff0ac80) {space_type: Parent, control_rig_element: {type: Control, name: \"llama_neck_01_ctrl\"}}>"
         1
        ]
       }
      ]
     }]
   }]
 }
 FortPlayerMannequin {
  id=#2
  possessed_class=FortPlayerMannequin
  tracks [
   {
    type=MovieScene3DTransformTrack
    sections [{
      type=MovieScene3DTransformSection
      channels [
       Location.X {type=MovieSceneScriptingDoubleChannel, default=0}
       Location.Y {type=MovieSceneScriptingDoubleChannel, default=0}
       Location.Z {type=MovieSceneScriptingDoubleChannel, default=0}
       Rotation.X {type=MovieSceneScriptingDoubleChannel, default=0}
       Rotation.Y {type=MovieSceneScriptingDoubleChannel, default=0}
       Rotation.Z {type=MovieSceneScriptingDoubleChannel, default=0}
       Scale.X {type=MovieSceneScriptingDoubleChannel, default=1}
       Scale.Y {type=MovieSceneScriptingDoubleChannel, default=1}
       Scale.Z {type=MovieSceneScriptingDoubleChannel, default=1}
      ]
     }]
   }
   {type=MovieSceneSkeletalAnimationTrack, sections []}
   {type=MovieSceneSpawnTrack, sections [{type=MovieSceneSpawnSection, channels [None {type=MovieSceneScriptingBoolChannel, default=true}]}]}
  ]
 }
]
display_rate {denominator=1, numerator=30}
tick_resolution {denominator=1, numerator=24000}
tracks []
