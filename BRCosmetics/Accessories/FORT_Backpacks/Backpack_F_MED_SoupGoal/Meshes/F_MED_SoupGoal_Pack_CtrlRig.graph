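# Control Rig for the SoupGoal backpack cosmetic, posing the skeleton of a llama character on the pack.
# Euler-transform FK controls mirror the bone hierarchy from spine base down through face, legs, digits, tail and breathing joint.
# On execution, one batch get reads the controls' global transforms and one batch set writes them to matching bones by array order, propagating to children.
# The graph contains no IK, constraints or simulation, despite the tail's "dyn" names.
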ControlRigBlueprint /BRCosmetics/Accessories/FORT_Backpacks/Backpack_F_MED_SoupGoal/Meshes/F_MED_SoupGoal_Pack_CtrlRig
graphs [
 RigVMFunctionLibrary {links [], nodes []}
 RigVMModel {nodes [
   RigUnit_BeginExecution = &RigUnit_BeginExecution()
   n1 = &RigUnit_ItemArray(Items="((Type=Control,Name=\"llama_spine_base_ctrl\"),(Type=Control,Name=\"llama_neck_01_ctrl\"),(Type=Control,Name=\"llama_neck_02_ctrl\"),(Type=Control,Name=\"llama_head_ctrl\"),(Type=Control,Name=\"llama_jaw_bottom_base_ctrl\"),(Type=Control,Name=\"llama_jaw_bottom_01_ct ...(truncated)")
   RigUnit_ItemArray = &RigUnit_ItemArray(Items="((Type=Bone,Name=\"llama_spine_base\"),(Type=Bone,Name=\"llama_neck_01\"),(Type=Bone,Name=\"llama_neck_02\"),(Type=Bone,Name=\"llama_head\"),(Type=Bone,Name=\"llama_jaw_bottom_base\"),(Type=Bone,Name=\"llama_jaw_bottom_01\"),(Type=Bone,Name=\"llama_tongue_1\"),(Type=Bon ...(truncated)")
   /Script/ControlRig.RigUnit_SetTransformItemArray(
    Space=GlobalSpace
    Weight=1
    bInitial=False
    bPropagateToChildren=true
    Transforms=/Script/ControlRig.RigUnit_GetTransformItemArray(Space=GlobalSpace, bInitial=False, Items=@n1.Items) {is=struct RigVMUnitNode}.Transforms
    ExecutePin=@RigUnit_BeginExecution.ExecutePin
    Items=@RigUnit_ItemArray.Items
   ) {is=struct RigVMUnitNode}
  ]}
]
hierarchy [
 root {loc=<0 0 0>, rot=<0 0 0 1>}
 pack {parent=root, loc=<0 -0.7316 0>, rot=<0 0 0 1>}
 llama_spine_base {parent=pack, loc=<-0.2225 -8.7893 -8.3719>, rot=<0 -0.7071 0 -0.7071>}
 llama_neck_01 {parent=llama_spine_base, loc=<-8.6013 0 -4.7356>, rot=<0 -0.9958 0 0.0914>}
 llama_neck_02 {parent=llama_neck_01, loc=<5.402 0 0.2162>, rot=<0 0.1089 0 -0.9941>}
 llama_head {parent=llama_neck_02, loc=<4.8063 0 0>, rot=<0 0 0 1>}
 llama_jaw_bottom_base {parent=llama_head, loc=<0.7236 0 -2.029>, rot=<0 -0.8862 0 0.4632>}
 llama_jaw_bottom_01 {parent=llama_jaw_bottom_base, loc=<10.7021 0 -0.564>, rot=<0 0 0 1>}
 llama_tongue_1 {parent=llama_jaw_bottom_base, loc=<3.3676 -0.0592 1.9021>, rot=<0 -0.0892 0 -0.996>}
 llama_tongue_2 {parent=llama_tongue_1, loc=<2 0 0>, rot=<0 0 0 1>}
 llama_tongue_3 {parent=llama_tongue_2, loc=<2 0 0>, rot=<0 -0.1926 0 -0.9813>}
 llama_tongue_4 {parent=llama_tongue_3, loc=<2 0 0>, rot=<0 0 0 1>}
 llama_jaw_top_base {parent=llama_head, loc=<0.7236 0 -2.029>, rot=<0 0.7067 0 -0.7075>}
 llama_jaw_top_01 {parent=llama_jaw_top_base, loc=<8.8194 0 0.7543>, rot=<0 0 0 1>}
 eye_l_base {parent=llama_jaw_top_base, loc=<6.3048 3.0552 -0.5667>, rot=<0.7577 0.6294 -0.0193 0.1713>}
 eye_r_base {parent=llama_jaw_top_base, loc=<6.1406 -3.1577 -0.6069>, rot=<0.0736 -0.9811 -0.1209 -0.1317>}
 llama_ear_r {parent=llama_head, loc=<3.2813 -1.6093 -2.6959>, rot=<0 -0.3898 0 -0.9209>}
 llama_ear_l {parent=llama_head, loc=<3.2813 1.5186 -2.6959>, rot=<0 -0.3898 0 -0.9209>}
 llama_smile_l {parent=llama_head, loc=<0.7253 3.4094 0>, rot=<0 -0.4775 0 -0.8786>}
 llama_smile_r {parent=llama_head, loc=<0.725 -3.409 0>, rot=<0.8786 0 -0.4775 0>}
 frnt_leg_l_01 {parent=llama_spine_base, loc=<9.9713 3.8227 -5.3488>, rot=<-0.0057 0.072 -0.078 -0.9943>}
 frnt_leg_l_02 {parent=frnt_leg_l_01, loc=<2.7197 0 -0.0478>, rot=<0 -0.4397 0 -0.8981>}
 frnt_finger_small_l_01 {parent=frnt_leg_l_02, loc=<2.8584 1.463 0.6198>, rot=<-0.215 0.6283 -0.1434 -0.7338>}
 frnt_finger_small_l_02 {parent=frnt_finger_small_l_01, loc=<1.5777 0 0>, rot=<0 0 0 1>}
 frnt_finger_mid_l_01 {parent=frnt_leg_l_02, loc=<3.2395 0.1962 -0.0629>, rot=<-0.0766 0.5373 0.0158 -0.8398>}
 frnt_finger_mid_l_02 {parent=frnt_finger_mid_l_01, loc=<2 0 0>, rot=<0 0 0 1>}
 frnt_finger_big_l_01 {parent=frnt_leg_l_02, loc=<2.9706 -1.31 -0.1148>, rot=<0.0767 0.5976 0.166 -0.7806>}
 frnt_finger_big_l_02 {parent=frnt_finger_big_l_01, loc=<1.7425 0 0>, rot=<0 0 0 1>}
 frnt_leg_r_01 {parent=llama_spine_base, loc=<9.9713 -3.8227 -5.3488>, rot=<0.0528 0.0595 0.0811 -0.9935>}
 frnt_leg_r_02 {parent=frnt_leg_r_01, loc=<2.7201 0 0>, rot=<0 0 0 1>}
 frnt_finger_big_r_01 {parent=frnt_leg_r_02, loc=<1.95 1.4424 -2.096>, rot=<-0.1716 0.1992 -0.1197 -0.9574>}
 frnt_finger_big_r_02 {parent=frnt_finger_big_r_01, loc=<1.4998 0 0>, rot=<0 0 0 1>}
 frnt_finger_mid_r_01 {parent=frnt_leg_r_02, loc=<1.988 -0.0049 -2.5719>, rot=<-0.0135 0.1229 -0.045 -0.9913>}
 frnt_finger_mid_r_02 {parent=frnt_finger_mid_r_01, loc=<2 0 0>, rot=<0 0 0 1>}
 frnt_finger_small_r_01 {parent=frnt_leg_r_02, loc=<2.2905 -1.3998 -1.9937>, rot=<0.1563 0.2502 0.0342 -0.9549>}
 frnt_finger_small_r_02 {parent=frnt_finger_small_r_01, loc=<1.5776 0 0>, rot=<0 0 0 1>}
 back_leg_l_01 {parent=llama_spine_base, loc=<9.3383 3.5224 6.1605>, rot=<-0.0246 0.2585 -0.0915 -0.9613>}
 back_leg_l_02 {parent=back_leg_l_01, loc=<3.724 0 0.0454>, rot=<0 -0.5236 0 -0.852>}
 bk_finger_small_l_01 {parent=back_leg_l_02, loc=<2.8321 1.2896 0.5296>, rot=<-0.2292 0.559 -0.1371 -0.785>}
 bk_finger_small_l_02 {parent=bk_finger_small_l_01, loc=<1.5059 0 0>, rot=<0 0 0 1>}
 bk_finger_mid_l_01 {parent=back_leg_l_02, loc=<3.1141 0.0454 -0.0707>, rot=<-0.0308 0.4501 0.0432 -0.8914>}
 bk_finger_mid_l_02 {parent=bk_finger_mid_l_01, loc=<1.8274 0 0>, rot=<0 0 0 1>}
 bk_finger_big_l_01 {parent=back_leg_l_02, loc=<2.9842 -1.3431 0.0144>, rot=<0.08 0.5609 0.1805 -0.804>}
 bk_finger_big_l_02 {parent=bk_finger_big_l_01, loc=<1.3583 0 0>, rot=<0 0 0 1>}
 back_leg_r_01 {parent=llama_spine_base, loc=<9.3383 -3.5225 6.1605>, rot=<-0.6275 0.257 -0.1104 -0.7267>}
 back_leg_r_02 {parent=back_leg_r_01, loc=<3.7243 0 0>, rot=<0 0 0 1>}
 bk_finger_big_r_01 {parent=back_leg_r_02, loc=<1.3402 -2.5503 -1.539>, rot=<0.5315 -0.0379 -0.116 -0.8382>}
 bk_finger_big_r_02 {parent=bk_finger_big_r_01, loc=<1.3583 0 0>, rot=<0 0 0 1>}
 bk_finger_mid_r_01 {parent=back_leg_r_02, loc=<1.3149 -2.8293 -0.1577>, rot=<0.6412 -0.1007 0.0252 -0.7603>}
 bk_finger_mid_r_02 {parent=bk_finger_mid_r_01, loc=<1.8274 0 0>, rot=<0 0 0 1>}
 bk_finger_small_r_01 {parent=back_leg_r_02, loc=<1.7239 -2.4012 1.1403>, rot=<0.8088 0.0377 -0.0461 -0.585>}
 bk_finger_small_r_02 {parent=bk_finger_small_r_01, loc=<1.5059 0 0>, rot=<0 0 0 1>}
 dyn_tail_base {parent=llama_spine_base, loc=<-3.5822 -0.0588 10.35>, rot=<0 -0.8661 0 0.4999>}
 dyn_tail_01 {parent=dyn_tail_base, loc=<0 0 0>, rot=<0 0 0 1>}
 dyn_tail_02 {parent=dyn_tail_01, loc=<3.0608 0 -0.1143>, rot=<0.0003 -0.4948 0.0668 -0.8664>}
 dyn_tail_03 {parent=dyn_tail_02, loc=<4.8311 0 0>, rot=<-0.0111 0.5782 -0.1037 -0.8092>}
 dyn_tail_04 {parent=dyn_tail_03, loc=<3.8843 0 0>, rot=<0 0 0 1>}
 llama_breathing_jnt {parent=llama_spine_base, loc=<1.6562 0 0>, rot=<0 0 0 1>}
 llama_spine_base_ctrl {controlType=EULER_TRANSFORM, type=Control, loc=<0 0 0>, rot=<0 0 0 1>}
 llama_neck_01_ctrl {controlType=EULER_TRANSFORM, parent=llama_spine_base_ctrl, type=Control, loc=<0 0 0>, rot=<0 0 0 1>}
 llama_neck_02_ctrl {controlType=EULER_TRANSFORM, parent=llama_neck_01_ctrl, type=Control, loc=<0 0 0>, rot=<0 0 0 1>}
 llama_head_ctrl {controlType=EULER_TRANSFORM, parent=llama_neck_02_ctrl, type=Control, loc=<0 0 0>, rot=<0 0 0 1>}
 llama_jaw_bottom_base_ctrl {controlType=EULER_TRANSFORM, parent=llama_head_ctrl, type=Control, loc=<0 0 0>, rot=<0 0 0 1>}
 llama_jaw_bottom_01_ctrl {controlType=EULER_TRANSFORM, parent=llama_jaw_bottom_base_ctrl, type=Control, loc=<0 0 0>, rot=<0 0 0 1>}
 llama_tongue_1_ctrl {controlType=EULER_TRANSFORM, parent=llama_jaw_bottom_base_ctrl, type=Control, loc=<0 0 0>, rot=<0 0 0 1>}
 llama_tongue_2_ctrl {controlType=EULER_TRANSFORM, parent=llama_tongue_1_ctrl, type=Control, loc=<0 0 0>, rot=<0 0 0 1>}
 llama_tongue_3_ctrl {controlType=EULER_TRANSFORM, parent=llama_tongue_2_ctrl, type=Control, loc=<0 0 0>, rot=<0 0 0 1>}
 llama_tongue_4_ctrl {controlType=EULER_TRANSFORM, parent=llama_tongue_3_ctrl, type=Control, loc=<0 0 0>, rot=<0 0 0 1>}
 llama_jaw_top_base_ctrl {controlType=EULER_TRANSFORM, parent=llama_head_ctrl, type=Control, loc=<0 0 0>, rot=<0 0 0 1>}
 llama_jaw_top_01_ctrl {controlType=EULER_TRANSFORM, parent=llama_jaw_top_base_ctrl, type=Control, loc=<0 0 0>, rot=<0 0 0 1>}
 eye_l_base_ctrl {controlType=EULER_TRANSFORM, parent=llama_jaw_top_base_ctrl, type=Control, loc=<0 0 0>, rot=<0 0 0 1>}
 eye_r_base_ctrl {controlType=EULER_TRANSFORM, parent=llama_jaw_top_base_ctrl, type=Control, loc=<0 0 0>, rot=<0 0 0 1>}
 llama_smile_l_ctrl {controlType=EULER_TRANSFORM, parent=llama_head_ctrl, type=Control, loc=<0 0 0>, rot=<0 0 0 1>}
 llama_smile_r_ctrl {controlType=EULER_TRANSFORM, parent=llama_head_ctrl, type=Control, loc=<0 0 0>, rot=<0 0 0 1>}
 llama_ear_r_ctrl {controlType=EULER_TRANSFORM, parent=llama_head_ctrl, type=Control, loc=<0 0 0>, rot=<0 0 0 1>}
 llama_ear_l_ctrl {controlType=EULER_TRANSFORM, parent=llama_head_ctrl, type=Control, loc=<0 0 0>, rot=<0 0 0 1>}
 frnt_leg_l_01_ctrl {controlType=EULER_TRANSFORM, parent=llama_spine_base_ctrl, type=Control, loc=<0 0 0>, rot=<0 0 0 1>}
 frnt_leg_l_02_ctrl {controlType=EULER_TRANSFORM, parent=frnt_leg_l_01_ctrl, type=Control, loc=<0 0 0>, rot=<0 0 0 1>}
 frnt_finger_small_l_01_ctrl {controlType=EULER_TRANSFORM, parent=frnt_leg_l_02_ctrl, type=Control, loc=<0 0 0>, rot=<0 0 0 1>}
 frnt_finger_small_l_02_ctrl {controlType=EULER_TRANSFORM, parent=frnt_finger_small_l_01_ctrl, type=Control, loc=<0 0 0>, rot=<0 0 0 1>}
 frnt_finger_mid_l_01_ctrl {controlType=EULER_TRANSFORM, parent=frnt_leg_l_02_ctrl, type=Control, loc=<0 0 0>, rot=<0 0 0 1>}
 frnt_finger_mid_l_02_ctrl {controlType=EULER_TRANSFORM, parent=frnt_finger_mid_l_01_ctrl, type=Control, loc=<0 0 0>, rot=<0 0 0 1>}
 frnt_finger_big_l_01_ctrl {controlType=EULER_TRANSFORM, parent=frnt_leg_l_02_ctrl, type=Control, loc=<0 0 0>, rot=<0 0 0 1>}
 frnt_finger_big_l_02_ctrl {controlType=EULER_TRANSFORM, parent=frnt_finger_big_l_01_ctrl, type=Control, loc=<0 0 0>, rot=<0 0 0 1>}
 frnt_leg_r_01_ctrl {controlType=EULER_TRANSFORM, parent=llama_spine_base_ctrl, type=Control, loc=<0 0 0>, rot=<0 0 0 1>}
 frnt_leg_r_02_ctrl {controlType=EULER_TRANSFORM, parent=frnt_leg_r_01_ctrl, type=Control, loc=<0 0 0>, rot=<0 0 0 1>}
 frnt_finger_big_r_01_ctrl {controlType=EULER_TRANSFORM, parent=frnt_leg_r_02_ctrl, type=Control, loc=<0 0 0>, rot=<0 0 0 1>}
 frnt_finger_big_r_02_ctrl {controlType=EULER_TRANSFORM, parent=frnt_finger_big_r_01_ctrl, type=Control, loc=<0 0 0>, rot=<0 0 0 1>}
 frnt_finger_mid_r_01_ctrl {controlType=EULER_TRANSFORM, parent=frnt_leg_r_02_ctrl, type=Control, loc=<0 0 0>, rot=<0 0 0 1>}
 frnt_finger_mid_r_02_ctrl {controlType=EULER_TRANSFORM, parent=frnt_finger_mid_r_01_ctrl, type=Control, loc=<0 0 0>, rot=<0 0 0 1>}
 frnt_finger_small_r_01_ctrl {controlType=EULER_TRANSFORM, parent=frnt_leg_r_02_ctrl, type=Control, loc=<0 0 0>, rot=<0 0 0 1>}
 frnt_finger_small_r_02_ctrl {controlType=EULER_TRANSFORM, parent=frnt_finger_small_r_01_ctrl, type=Control, loc=<0 0 0>, rot=<0 0 0 1>}
 back_leg_l_01_ctrl {controlType=EULER_TRANSFORM, parent=llama_spine_base_ctrl, type=Control, loc=<0 0 0>, rot=<0 0 0 1>}
 back_leg_l_02_ctrl {controlType=EULER_TRANSFORM, parent=back_leg_l_01_ctrl, type=Control, loc=<0 0 0>, rot=<0 0 0 1>}
 bk_finger_small_l_01_ctrl {controlType=EULER_TRANSFORM, parent=back_leg_l_02_ctrl, type=Control, loc=<0 0 0>, rot=<0 0 0 1>}
 bk_finger_small_l_02_ctrl {controlType=EULER_TRANSFORM, parent=bk_finger_small_l_01_ctrl, type=Control, loc=<0 0 0>, rot=<0 0 0 1>}
 bk_finger_mid_l_01_ctrl {controlType=EULER_TRANSFORM, parent=back_leg_l_02_ctrl, type=Control, loc=<0 0 0>, rot=<0 0 0 1>}
 bk_finger_mid_l_02_ctrl {controlType=EULER_TRANSFORM, parent=bk_finger_mid_l_01_ctrl, type=Control, loc=<0 0 0>, rot=<0 0 0 1>}
 bk_finger_big_l_01_ctrl {controlType=EULER_TRANSFORM, parent=back_leg_l_02_ctrl, type=Control, loc=<0 0 0>, rot=<0 0 0 1>}
 bk_finger_big_l_02_ctrl {controlType=EULER_TRANSFORM, parent=bk_finger_big_l_01_ctrl, type=Control, loc=<0 0 0>, rot=<0 0 0 1>}
 back_leg_r_01_ctrl {controlType=EULER_TRANSFORM, parent=llama_spine_base_ctrl, type=Control, loc=<0 0 0>, rot=<0 0 0 1>}
 back_leg_r_02_ctrl {controlType=EULER_TRANSFORM, parent=back_leg_r_01_ctrl, type=Control, loc=<0 0 0>, rot=<0 0 0 1>}
 bk_finger_big_r_01_ctrl {controlType=EULER_TRANSFORM, parent=back_leg_r_02_ctrl, type=Control, loc=<0 0 0>, rot=<0 0 0 1>}
 bk_finger_big_r_02_ctrl {controlType=EULER_TRANSFORM, parent=bk_finger_big_r_01_ctrl, type=Control, loc=<0 0 0>, rot=<0 0 0 1>}
 bk_finger_mid_r_01_ctrl {controlType=EULER_TRANSFORM, parent=back_leg_r_02_ctrl, type=Control, loc=<0 0 0>, rot=<0 0 0 1>}
 bk_finger_mid_r_02_ctrl {controlType=EULER_TRANSFORM, parent=bk_finger_mid_r_01_ctrl, type=Control, loc=<0 0 0>, rot=<0 0 0 1>}
 bk_finger_small_r_01_ctrl {controlType=EULER_TRANSFORM, parent=back_leg_r_02_ctrl, type=Control, loc=<0 0 0>, rot=<0 0 0 1>}
 bk_finger_small_r_02_ctrl {controlType=EULER_TRANSFORM, parent=bk_finger_small_r_01_ctrl, type=Control, loc=<0 0 0>, rot=<0 0 0 1>}
 dyn_tail_base_ctrl {controlType=EULER_TRANSFORM, parent=llama_spine_base_ctrl, type=Control, loc=<0 0 0>, rot=<0 0 0 1>}
 dyn_tail_01_ctrl {controlType=EULER_TRANSFORM, parent=dyn_tail_base_ctrl, type=Control, loc=<0 0 0>, rot=<0 0 0 1>}
 dyn_tail_02_ctrl {controlType=EULER_TRANSFORM, parent=dyn_tail_01_ctrl, type=Control, loc=<0 0 0>, rot=<0 0 0 1>}
 dyn_tail_03_ctrl {controlType=EULER_TRANSFORM, parent=dyn_tail_02_ctrl, type=Control, loc=<0 0 0>, rot=<0 0 0 1>}
 dyn_tail_04_ctrl {controlType=EULER_TRANSFORM, parent=dyn_tail_03_ctrl, type=Control, loc=<0 0 0>, rot=<0 0 0 1>}
 llama_breathing_jnt_ctrl {controlType=EULER_TRANSFORM, parent=llama_spine_base_ctrl, type=Control, loc=<0 0 0>, rot=<0 0 0 1>}
]
