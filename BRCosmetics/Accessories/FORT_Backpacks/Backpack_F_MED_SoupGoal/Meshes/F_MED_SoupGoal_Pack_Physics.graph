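# Physics asset for the SoupGoal back bling mesh, defining collision and joints for its dynamic tail bone chain.
# A large root capsule anchors a small tail base, followed by four linked capsule segments.
# The base rotates freely against the root.
# The segments have narrow swing limits and zero twist limits, with one swing axis widening toward the tip.
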PhysicsAsset /BRCosmetics/Accessories/FORT_Backpacks/Backpack_F_MED_SoupGoal/Meshes/F_MED_SoupGoal_Pack_Physics
bodies {
 dyn_tail_01 [{
   shapeType=Capsule
   length=2.4762
   radius=1.673
   shapeName=dyn_tail_01_capsule
   center=<1.4752 -0.0215 0.6934>
   rotation=<rot -6.8463 -90 34.2557>
  }]
 dyn_tail_02 [{
   shapeType=Capsule
   length=4.4614
   radius=2.5931
   shapeName=dyn_tail_02_capsule
   center=<2.0125 0.1379 -0.348>
   rotation=<rot 0 -101.0895 -63.1994>
  }]
 dyn_tail_03 [{
   shapeType=Capsule
   length=3.2798
   radius=2.1337
   shapeName=dyn_tail_03_capsule
   center=<1.974 -0.1048 -0.5144>
   rotation=<rot 0 86.0249 -83.5516>
  }]
 dyn_tail_04 [{
   shapeType=Capsule
   length=3.3743
   radius=1.5047
   shapeName=dyn_tail_04_capsule
   center=<1.6357 -0.0056 2.6276>
   rotation=<rot -7.0689 -90 27.2592>
  }]
 dyn_tail_base [{shapeType=Capsule, length=0.505, radius=0.505, shapeName=dyn_tail_base_capsule, rotation=<rot 0 -90 0>}]
 root [{shapeType=Capsule, length=20, radius=20, shapeName=root_capsule, center=<0 -10 0>, rotation=<rot 0 -90 0>}]
}
constraints [
 {bone1Name=dyn_tail_01, bone2Name=dyn_tail_base, swing1LimitDegrees=5, swing2LimitDegrees=5, twistLimitDegrees=0}
 {bone1Name=dyn_tail_02, bone2Name=dyn_tail_01, swing1LimitDegrees=5, swing2LimitDegrees=10, twistLimitDegrees=0}
 {bone1Name=dyn_tail_03, bone2Name=dyn_tail_02, swing1LimitDegrees=5, swing2LimitDegrees=10, twistLimitDegrees=0}
 {bone1Name=dyn_tail_04, bone2Name=dyn_tail_03, swing1LimitDegrees=5, swing2LimitDegrees=20, twistLimitDegrees=0}
 {
  bone1Name=dyn_tail_base
  bone2Name=root
  swing1LimitDegrees=45
  swing1Motion=Free
  swing2LimitDegrees=45
  swing2Motion=Free
  twistLimitDegrees=45
  twistMotion=Free
 }
]
